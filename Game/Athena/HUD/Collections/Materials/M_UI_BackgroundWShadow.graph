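# Translucent UI material for the HUD Collections that procedurally draws a background panel with an angled rarity band and offset shadow.
# Scaled, rotated UVs and smoothstepped edges form the panel mask; a second rotated edge cuts the RarityColor band within it.
# The mask logic repeats at ShadowOffset, applied diagonally from its R channel, and subtracts the panel to leave the shadow.
# Lerps build color, Max merges opacities, and GlobalScale scales and offsets UVs.
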Material /Game/Athena/HUD/Collections/Materials/M_UI_BackgroundWShadow
graph {
 properties {blendMode=BLEND_Translucent, materialDomain=MD_UI, numCustomizedUVs=1, pixelDepthOffsetMode=PDOM_Legacy, usedShadingModels=Default Lit}
 nodes [
  n1 = VectorParameter() {parameterName=GlobalScale, defaultValue=<rgba 1 1 0 0>}
  Add(
   A=MaterialFunctionCall(Texture Scale=AppendVector(A=@n1.R, B=@n1.R)) {materialFunction=&ScaleUVsByCenter}.UVs
   B=AppendVector(A=@n1.B, B=@n1.B)
  )
  n2 = VectorParameter() {parameterName=ShadowColor, defaultValue=<rgba 1 0 0.2527 0.4>}
  n3 = VectorParameter() {parameterName=FillColor, defaultValue=<rgba 0 0.9877 1 0.5>}
  n4 = TextureCoordinate()
  n5 = VectorParameter() {parameterName="Texture Scale (V2)", defaultValue=<rgba 1 0.8 0 0>}
  n6 = ScalarParameter() {parameterName=SideAngle, sliderMax=1}
  n7 = ComponentMask(None=Multiply(
    A=MaterialFunctionCall(UVs=@n4, Texture Scale=@n5.RGB) {materialFunction=&ScaleUVsByCenter}.0-1 mask
    B=MaterialFunctionCall(UVs=@n4, "Rotation Angle (0-1)"=@n6) {materialFunction=&CustomRotator}.Rotated Values
   )) {r=true}
  n8 = ScalarParameter() {defaultValue=0.1, parameterName=LeftEdge}
  n9 = Multiply(A=@n8, B=ScalarParameter() {defaultValue=0.1, parameterName=EdgeSoftness})
  n10 = ScalarParameter() {defaultValue=0.9, parameterName=RightEdge}
  n11 = Multiply(A=@n10, B=ScalarParameter() {defaultValue=0.1, parameterName=EdgeSoftness})
  n12 = Subtract(
   A=MaterialFunctionCall(Alpha=@n7, Min=@n8, Max=Add(A=@n8, B=@n9)) {materialFunction=&SmoothStep}.Result
   B=MaterialFunctionCall(Alpha=@n7, Min=@n10, Max=Add(A=@n10, B=@n11)) {materialFunction=&SmoothStep}.Result
  )
  n13 = ScalarParameter() {parameterName=RarityAngle, sliderMax=1, sliderMin=-1}
  n14 = ScalarParameter() {defaultValue=0.85, parameterName=RarityEdge}
  n15 = Multiply(A=@n14, B=ScalarParameter() {defaultValue=0.1, parameterName=EdgeSoftness})
  n16 = Multiply(
   A=@n12
   B=MaterialFunctionCall(
    Alpha=MaterialFunctionCall(UVs=@n4, "Rotation Angle (0-1)"=@n13) {materialFunction=&CustomRotator}.Rotated Values
    Min=@n14
    Max=Add(A=@n14, B=@n15)
   ) {materialFunction=&SmoothStep}.Result
  )
  LinearInterpolate(
   A=LinearInterpolate(A=@n2.RGB, B=@n3.RGB, Alpha=@n12)
   B=VectorParameter() {parameterName=RarityColor, defaultValue=<rgba 0 0.1023 1 1>}.RGB
   Alpha=@n16
  )
  n17 = VectorParameter() {parameterName=ShadowOffset, defaultValue=<rgba -0.1 -0.1 0 1>}
  n18 = Add(A=@n4, B=AppendVector(A=@n17.R, B=@n17.R))
  n19 = VectorParameter() {parameterName="Texture Scale (V2)", defaultValue=<rgba 1 0.8 0 0>}
  n20 = ScalarParameter() {parameterName=SideAngle, sliderMax=1}
  n21 = ComponentMask(None=Multiply(
    A=MaterialFunctionCall(UVs=@n18, Texture Scale=@n19.RGB) {materialFunction=&ScaleUVsByCenter}.0-1 mask
    B=MaterialFunctionCall(UVs=@n18, "Rotation Angle (0-1)"=@n20) {materialFunction=&CustomRotator}.Rotated Values
   )) {r=true}
  n22 = ScalarParameter() {defaultValue=0.1, parameterName=LeftEdge}
  n23 = Multiply(A=@n22, B=ScalarParameter() {defaultValue=0.1, parameterName=EdgeSoftness})
  n24 = ScalarParameter() {defaultValue=0.9, parameterName=RightEdge}
  n25 = Multiply(A=@n24, B=ScalarParameter() {defaultValue=0.1, parameterName=EdgeSoftness})
  n26 = Subtract(
   A=MaterialFunctionCall(Alpha=@n21, Min=@n22, Max=Add(A=@n22, B=@n23)) {materialFunction=&SmoothStep}.Result
   B=MaterialFunctionCall(Alpha=@n21, Min=@n24, Max=Add(A=@n24, B=@n25)) {materialFunction=&SmoothStep}.Result
  )
  Max(A=Max(A=Multiply(A=@n12, B=@n3.A), B=@n16), B=Multiply(A=Subtract(A=@n26, B=@n12), B=@n2.A))
 ]
}
