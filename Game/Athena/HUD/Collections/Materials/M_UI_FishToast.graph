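# Translucent UI material for the fish toast in the Collections HUD, styling an icon texture that defaults to the Flopper placement icon.
# Two rotated UV gradients, Lowlight and Highlight, multiply into the icon's color, with each tint's alpha setting its angle and a Whiteout fade toward white.
# An offset resample of the icon's alpha forms a drop shadow behind it.
# An optional radial gradient masks the combined opacity to a circle.
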Material /Game/Athena/HUD/Collections/Materials/M_UI_FishToast
graph {
 properties {blendMode=BLEND_Translucent, materialDomain=MD_UI, pixelDepthOffsetMode=PDOM_Legacy, usedShadingModels=Default Lit}
 nodes [
  n1 = VectorParameter() {parameterName=Shadow, defaultValue=<rgba 0.0052 0.0052 0.0056 0.5>}
  n2 = VectorParameter() {parameterName=Lowlight, defaultValue=<rgba 1 1 1 0.1961>}
  n3 = ComponentMask(None=MaterialFunctionCall(UVs=TextureCoordinate(), "Rotation Angle (0-1)"=@n2.A) {materialFunction=&CustomRotator}.Rotated Values) {r=true}
  n4 = ScalarParameter() {parameterName=ShadowExp}
  n5 = TextureSampleParameter2D(UVs=TextureCoordinate()) {
   parameterName=Texture
   samplerSource=SSM_Clamp_WorldGroupSettings
   texture=/Game/Athena/Items/EnvironmentalItems/FlopperSpawn/T_Athena_FlopperPlacementIcon
  }
  n6 = Multiply(A=Multiply(A=@n2.RGB, B=Power(Base=@n3, Exp=@n4)), B=@n5.RGB)
  n7 = VectorParameter() {parameterName=Highlight, defaultValue=<rgba 1 1 1 0.1961>}
  n8 = MaterialFunctionCall(UVs=TextureCoordinate(), "Rotation Angle (0-1)"=@n7.A) {materialFunction=&CustomRotator}
  n9 = Multiply(A=@n7.RGB, B=Power(Base=ComponentMask(None=@n8.Rotated Values) {r=true}, Exp=ScalarParameter() {parameterName=LightingExp}))
  LinearInterpolate(A=@n1.RGB, B=LinearInterpolate(A=Multiply(A=@n6, B=@n9), B=Constant() {r=1}, Alpha=ScalarParameter() {parameterName=Whiteout}), Alpha=@n5.A)
  n10 = TextureSampleParameter2D(UVs=Add(A=TextureCoordinate(), B=ScalarParameter() {defaultValue=-0.0381, parameterName=ShadowOffset})) {
   parameterName=Texture
   samplerSource=SSM_Clamp_WorldGroupSettings
   texture=&T_Athena_FlopperPlacementIcon
  }
  n11 = ScalarParameter() {defaultValue=128, parameterName=CircleDensity}
  Multiply(
   A=Max(A=Multiply(A=@n10.A, B=@n1.A), B=@n5.A)
   B=LinearInterpolate(
    B=MaterialFunctionCall(Density=@n11) {materialFunction=&RadialGradientExponential}.RadialGradientExponential
    Alpha=ScalarParameter() {parameterName=UseCircle}
   ) {constA=1}
  )
 ]
}
