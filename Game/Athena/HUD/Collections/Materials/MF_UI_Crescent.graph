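# UI material function that procedurally draws a crescent mask for HUD collections materials.
# It thresholds two exponential radial gradients with SmoothStep, the second sampled at UVs shifted by CrescentOffset, then subtracts that offset disc from the first.
# OuterEdge and MaskEdge set the two disc thresholds, and a shared EdgeSoftness widens each transition in proportion to its edge value.
# The subtraction is output directly, without a clamp.
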MaterialFunction /Game/Athena/HUD/Collections/Materials/MF_UI_Crescent
graph {nodes [
  n1 = FunctionInput(Preview=TextureCoordinate()) {bUsePreviewValueAsDefault=true, inputName=UV, inputType=FunctionInput_Vector2}
  n2 = MaterialFunctionCall(UVs=@n1) {materialFunction=&RadialGradientExponential}
  n3 = FunctionInput(Preview=Constant() {r=0.1}) {bUsePreviewValueAsDefault=true, inputName=OuterEdge, inputType=FunctionInput_Scalar}
  n4 = FunctionInput(Preview=Constant() {r=0.01}) {bUsePreviewValueAsDefault=true, inputName=EdgeSoftness, inputType=FunctionInput_Scalar}
  n5 = Add(A=@n3, B=Multiply(A=@n3, B=@n4))
  n6 = Constant2Vector() {g=0.05, r=0.051}
  n7 = MaterialFunctionCall(UVs=Add(A=@n1, B=FunctionInput(Preview=@n6) {bUsePreviewValueAsDefault=true, inputName=CrescentOffset, inputType=FunctionInput_Vector2})) {materialFunction=&RadialGradientExponential}
  n8 = FunctionInput(Preview=Constant() {r=0.2}) {bUsePreviewValueAsDefault=true, inputName=MaskEdge, inputType=FunctionInput_Scalar}
  n9 = Add(A=@n8, B=Multiply(A=@n8, B=@n4))
  FunctionOutput(None=Subtract(
    A=MaterialFunctionCall(Alpha=@n2.RadialGradientExponential, Min=@n3, Max=@n5) {materialFunction=&SmoothStep}.Result
    B=MaterialFunctionCall(Alpha=@n7.RadialGradientExponential, Min=@n8, Max=@n9) {materialFunction=&SmoothStep}.Result
   ))
 ]}
