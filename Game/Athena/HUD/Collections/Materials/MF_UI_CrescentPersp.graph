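# UI material function that draws an animated, perspective-warped crescent mask, stored with the Athena HUD collection materials.
# Offset UVs pass through MF_UI_BasicPerspective, then a Rotator spins them into MF_UI_Crescent, which takes edge, softness and offset inputs.
# Rotation uses looping Time plus TimeOffset, scaled by TimeSpeed, unless ManualTime is positive, when ManualTime times TimeOffset drives it.
# The crescent is multiplied by the perspective function's 0-or-1 mask and saturated.
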MaterialFunction /Game/Athena/HUD/Collections/Materials/MF_UI_CrescentPersp
graph {nodes [
  n1 = TextureCoordinate()
  n2 = Constant2Vector()
  n3 = MaterialFunctionCall!(
   UV=Add(
    A=FunctionInput(Preview=@n1) {bUsePreviewValueAsDefault=true, inputName=UV, inputType=FunctionInput_Vector2}
    B=FunctionInput(Preview=@n2) {bUsePreviewValueAsDefault=true, inputName=ObjectOffset, inputType=FunctionInput_Vector2, sortPriority=1}
   )
   Texture
   FixedScaleFactor_X=FunctionInput(Preview=Constant()) {bUsePreviewValueAsDefault=true, inputName="FixedScaleFactor_X ", inputType=FunctionInput_Scalar, sortPriority=2}
   FixedScaleFactor_Y=FunctionInput(Preview=Constant()) {bUsePreviewValueAsDefault=true, inputName=FixedScaleFactor_Y, inputType=FunctionInput_Scalar, sortPriority=3}
   Persp_Y=FunctionInput(Preview=Constant()) {bUsePreviewValueAsDefault=true, inputName=Persp_Y, inputType=FunctionInput_Scalar, sortPriority=4}
   Persp_X=FunctionInput(Preview=Constant()) {bUsePreviewValueAsDefault=true, inputName=Persp_X, inputType=FunctionInput_Scalar, sortPriority=5}
   Slide_X=FunctionInput(Preview=Constant()) {bUsePreviewValueAsDefault=true, inputName=Slide_X, inputType=FunctionInput_Scalar, sortPriority=6}
   Slide_Y=FunctionInput(Preview=Constant()) {bUsePreviewValueAsDefault=true, inputName=Slide_Y, inputType=FunctionInput_Scalar, sortPriority=7}
   Flip_Y
   Flip_X
  ) {materialFunction=/Game/UI/Foundation/Materials/MF_UI_BasicPerspective}
  n4 = FunctionInput(Preview=Constant()) {bUsePreviewValueAsDefault=true, inputName=ManualTime, inputType=FunctionInput_Scalar, sortPriority=8}
  n5 = Constant()
  n6 = FunctionInput(Preview=Constant()) {bUsePreviewValueAsDefault=true, inputName=TimeOffset, inputType=FunctionInput_Scalar, sortPriority=9}
  n7 = Multiply(A=@n4, B=@n6)
  n8 = Multiply(
   A=Add(A=@n6, B=Time() {period=1})
   B=FunctionInput(Preview=Constant() {r=2}) {bUsePreviewValueAsDefault=true, inputName=TimeSpeed, inputType=FunctionInput_Scalar, sortPriority=9}
  )
  n9 = MaterialFunctionCall!(
   EdgeSoftness=FunctionInput(Preview=Constant() {r=0.2}) {bUsePreviewValueAsDefault=true, inputName=CrescentEdgeSoftness, inputType=FunctionInput_Scalar, sortPriority=10}
   CrescentOffset=FunctionInput(Preview=Constant3Vector() {constant=<rgba 0.05 0.05 0 0>}) {bUsePreviewValueAsDefault=true, inputName=CrescentOffset, inputType=FunctionInput_Vector2, sortPriority=11}
   MaskEdge=FunctionInput(Preview=Constant() {r=0.1}) {bUsePreviewValueAsDefault=true, inputName=CrescentEdge, inputType=FunctionInput_Scalar, sortPriority=12}
   UV=Rotator(Coordinate=@n3.FinalUvs, Time=If(A=@n4, B=@n5, "A > B"=@n7, "A == B"=@n8, "A < B"=@n8)) {speed=1.1653}
   OuterEdge
  ) {materialFunction=&/MF_UI_Crescent}
  FunctionOutput(None=Saturate(None=Multiply(A=@n3.0 or 1 Mask, B=@n9.Result)))
 ]}
