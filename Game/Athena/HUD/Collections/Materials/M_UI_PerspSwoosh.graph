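# Translucent HUD UI material that builds four animated crescent "swooshes" with perspective.
# Four MF_UI_CrescentPersp calls, each with its own scale, perspective, slide, speed and time offset, share ManualTime and merge through Max.
# All four reuse Swoosh1's edge softness and crescent offset, leaving the numbered variants unconnected.
# A separate single-swoosh chain (BasicPerspective, Rotator, Crescent) and a radial gradient feed nothing shown.
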Material /Game/Athena/HUD/Collections/Materials/M_UI_PerspSwoosh
graph {
 properties {blendMode=BLEND_Translucent, materialDomain=MD_UI, pixelDepthOffsetMode=PDOM_Legacy, usedShadingModels=Default Lit}
 nodes [
  n1 = VectorParameter() {group=Swoosh1, parameterName=Swoosh1Offset, defaultValue=<rgba 0 0 0 1>}
  n2 = VectorParameter() {group=Swoosh2, parameterName=2Swoosh1Offset, defaultValue=<rgba 0 0 0 1>}
  n3 = Max(
   A=MaterialFunctionCall!(
    UV=TextureCoordinate()
    ObjectOffset=AppendVector(A=@n1.R, B=@n1.R)
    "FixedScaleFactor_X "=ScalarParameter() {group=Swoosh1, parameterName="FixedScaleFactor_X (S)"}
    FixedScaleFactor_Y=ScalarParameter() {group=Swoosh1, parameterName="FixedScaleFactor_Y (S)"}
    Persp_Y=ScalarParameter() {group=Swoosh1, parameterName="Persp_Y (S)"}
    Persp_X=ScalarParameter() {group=Swoosh1, parameterName="Persp_X (S)"}
    Slide_X=ScalarParameter() {group=Swoosh1, parameterName="Slide_X (S)"}
    Slide_Y=ScalarParameter() {group=Swoosh1, parameterName="Slide_Y (S)"}
    ManualTime=ScalarParameter() {parameterName=ManualTime}
    TimeSpeed=ScalarParameter() {defaultValue=2, group=Swoosh1, parameterName=Speed}
    TimeOffset=ScalarParameter() {defaultValue=1, group=Swoosh1, parameterName=TimeOffset}
    CrescentEdgeSoftness=ScalarParameter() {defaultValue=0.2, group=Swoosh1, parameterName="EdgeSoftness (S)"}
    CrescentOffset=VectorParameter() {group=Swoosh1, parameterName="CrescentOffset (V2)", defaultValue=<rgba 0.05 0.05 0 0>}.RGB
    CrescentEdge=ScalarParameter() {defaultValue=0.1, group=Swoosh1, parameterName="MaskEdge (S)"}
   ) {materialFunction=&/MF_UI_CrescentPersp}.Result
   B=MaterialFunctionCall(
    UV=TextureCoordinate()
    ObjectOffset=AppendVector(A=@n2.R, B=@n2.R)
    "FixedScaleFactor_X "=ScalarParameter() {group=Swoosh2, parameterName="2FixedScaleFactor_X (S)"}
    FixedScaleFactor_Y=ScalarParameter() {group=Swoosh2, parameterName="2FixedScaleFactor_Y (S)"}
    Persp_Y=ScalarParameter() {group=Swoosh2, parameterName="2Persp_Y (S)"}
    Persp_X=ScalarParameter() {group=Swoosh2, parameterName="2Persp_X (S)"}
    Slide_X=ScalarParameter() {group=Swoosh2, parameterName="2Slide_X (S)"}
    Slide_Y=ScalarParameter() {group=Swoosh2, parameterName="2Slide_Y (S)"}
    ManualTime=ScalarParameter() {parameterName=ManualTime}
    TimeSpeed=ScalarParameter() {defaultValue=2, group=Swoosh2, parameterName=2Speed}
    TimeOffset=ScalarParameter() {defaultValue=0.2, group=Swoosh2, parameterName=2TimeOffset}
    CrescentEdgeSoftness=ScalarParameter() {defaultValue=0.2, group=Swoosh1, parameterName="EdgeSoftness (S)"}
    CrescentOffset=VectorParameter() {group=Swoosh1, parameterName="CrescentOffset (V2)", defaultValue=<rgba 0.05 0.05 0 0>}.RGB
    CrescentEdge=ScalarParameter() {defaultValue=0.1, group=Swoosh2, parameterName="2MaskEdge (S)"}
   ) {materialFunction=&MF_UI_CrescentPersp}.Result
  )
  n4 = VectorParameter() {group=Swoosh3, parameterName=3Swoosh1Offset, defaultValue=<rgba 0 0 0 1>}
  n5 = MaterialFunctionCall(
   UV=TextureCoordinate()
   ObjectOffset=AppendVector(A=@n4.R, B=@n4.R)
   "FixedScaleFactor_X "=ScalarParameter() {group=Swoosh3, parameterName="3FixedScaleFactor_X (S)"}
   FixedScaleFactor_Y=ScalarParameter() {group=Swoosh3, parameterName="3FixedScaleFactor_Y (S)"}
   Persp_Y=ScalarParameter() {group=Swoosh3, parameterName="3Persp_Y (S)"}
   Persp_X=ScalarParameter() {group=Swoosh3, parameterName="3Persp_X (S)"}
   Slide_X=ScalarParameter() {group=Swoosh3, parameterName="3Slide_X (S)"}
   Slide_Y=ScalarParameter() {group=Swoosh3, parameterName="3Slide_Y (S)"}
   ManualTime=ScalarParameter() {parameterName=ManualTime}
   TimeSpeed=ScalarParameter() {defaultValue=2, group=Swoosh3, parameterName=3Speed}
   TimeOffset=ScalarParameter() {defaultValue=0.4, group=Swoosh3, parameterName=3TimeOffset}
   CrescentEdgeSoftness=ScalarParameter() {defaultValue=0.2, group=Swoosh1, parameterName="EdgeSoftness (S)"}
   CrescentOffset=VectorParameter() {group=Swoosh1, parameterName="CrescentOffset (V2)", defaultValue=<rgba 0.05 0.05 0 0>}.RGB
   CrescentEdge=ScalarParameter() {defaultValue=0.1, group=Swoosh3, parameterName="3MaskEdge (S)"}
  ) {materialFunction=&MF_UI_CrescentPersp}
  n6 = TextureCoordinate()
  n7 = VectorParameter() {group=Swoosh4, parameterName=4Swoosh1Offset, defaultValue=<rgba 0 0 0 1>}
  n8 = AppendVector(A=@n7.R, B=@n7.R)
  n9 = ScalarParameter() {group=Swoosh4, parameterName="4FixedScaleFactor_X (S)"}
  n10 = ScalarParameter() {group=Swoosh4, parameterName="4FixedScaleFactor_Y (S)"}
  n11 = ScalarParameter() {group=Swoosh4, parameterName="4Persp_Y (S)"}
  n12 = ScalarParameter() {group=Swoosh4, parameterName="4Persp_X (S)"}
  n13 = ScalarParameter() {group=Swoosh4, parameterName="4Slide_X (S)"}
  n14 = ScalarParameter() {group=Swoosh4, parameterName="4Slide_Y (S)"}
  n15 = ScalarParameter() {parameterName=ManualTime}
  n16 = ScalarParameter() {defaultValue=2, group=Swoosh4, parameterName=4Speed}
  n17 = ScalarParameter() {defaultValue=0.6, group=Swoosh4, parameterName=4TimeOffset}
  n18 = ScalarParameter() {defaultValue=0.2, group=Swoosh1, parameterName="EdgeSoftness (S)"}
  n19 = VectorParameter() {group=Swoosh1, parameterName="CrescentOffset (V2)", defaultValue=<rgba 0.05 0.05 0 0>}
  n20 = ScalarParameter() {defaultValue=0.1, group=Swoosh4, parameterName="4MaskEdge (S)"}
  Add(B=Max(
    A=Max(A=@n3, B=@n5.Result)
    B=MaterialFunctionCall(
     UV=@n6
     ObjectOffset=@n8
     "FixedScaleFactor_X "=@n9
     FixedScaleFactor_Y=@n10
     Persp_Y=@n11
     Persp_X=@n12
     Slide_X=@n13
     Slide_Y=@n14
     ManualTime=@n15
     TimeSpeed=@n16
     TimeOffset=@n17
     CrescentEdgeSoftness=@n18
     CrescentOffset=@n19.RGB
     CrescentEdge=@n20
    ) {materialFunction=&MF_UI_CrescentPersp}.Result
   ))
  Constant() {r=1}
  MaterialFunctionCall(Radius=Constant() {r=0.1}) {materialFunction=&RadialGradientExponential}
  n21 = VectorParameter() {parameterName=Swoosh1Offset, defaultValue=<rgba 0 0 0 1>}
  n22 = MaterialFunctionCall!(
   UV=Add(A=TextureCoordinate(), B=AppendVector(A=@n21.R, B=@n21.R))
   Texture
   FixedScaleFactor_X=ScalarParameter() {parameterName="FixedScaleFactor_X (S)"}
   FixedScaleFactor_Y=ScalarParameter() {parameterName="FixedScaleFactor_Y (S)"}
   Persp_Y=ScalarParameter() {parameterName="Persp_Y (S)"}
   Persp_X=ScalarParameter() {parameterName="Persp_X (S)"}
   Slide_X=ScalarParameter() {parameterName="Slide_X (S)"}
   Slide_Y=ScalarParameter() {parameterName="Slide_Y (S)"}
   Flip_Y
   Flip_X
  ) {materialFunction=/Game/UI/Foundation/Materials/MF_UI_BasicPerspective}
  n23 = ScalarParameter() {defaultValue=0.2, parameterName="EdgeSoftness (S)"}
  n24 = VectorParameter() {parameterName="CrescentOffset (V2)", defaultValue=<rgba 0.05 0.05 0 0>}
  n25 = ScalarParameter() {defaultValue=0.1, parameterName="MaskEdge (S)"}
  n26 = Rotator(Coordinate=@n22.FinalUvs, Time=Multiply(A=Time() {period=1}, B=ScalarParameter() {defaultValue=2, parameterName=Speed})) {speed=1.1653}
  Saturate(None=Multiply(
    A=@n22.0 or 1 Mask
    B=MaterialFunctionCall!(EdgeSoftness=@n23, CrescentOffset=@n24.RGB, MaskEdge=@n25, UV=@n26, OuterEdge) {materialFunction=&/MF_UI_Crescent}.Result
   ))
  ScalarParameter() {defaultValue=0.2, group=Swoosh2, parameterName="2EdgeSoftness (S)"}
  ScalarParameter() {defaultValue=0.2, group=Swoosh3, parameterName="3EdgeSoftness (S)"}
  ScalarParameter() {defaultValue=0.2, group=Swoosh4, parameterName="4EdgeSoftness (S)"}
  VectorParameter() {group=Swoosh2, parameterName="2CrescentOffset (V2)", defaultValue=<rgba 0.05 0.05 0 0>}
  VectorParameter() {group=Swoosh3, parameterName="3CrescentOffset (V2)", defaultValue=<rgba 0.05 0.05 0 0>}
  VectorParameter() {group=Swoosh4, parameterName="4CrescentOffset (V2)", defaultValue=<rgba 0.05 0.05 0 0>}
 ]
}
