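# NPC_Toast_Collected_Cue is a Season 15 UI sound cue, named for the toast shown when an NPC is collected.
# A single wave player feeds the NPC_Toast_Collected wave straight to the output in the Fort_UI sound class at reduced volume.
# A concurrency override allows one instance at a time, stopping the oldest when a new one starts.
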SoundCue /Game/Sounds/Fort_UI_Sounds/S15/NPC_Toast_Collected_Cue
root=WavePlayer_1
nodes [{id=WavePlayer_1, properties {soundWaveAssetPtr=&/NPC_Toast_Collected}}]
properties {
 bOverrideConcurrency=true
 maxDistance=2097152
 soundClassObject=&Fort_UI
 volumeMultiplier=0.55
 concurrencyOverrides {maxCount=1, resolutionRule=StopOldest}
}
