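# Character variant of the collection toast widget, showing an NPC's image, number, name, location count and a collection prompt.
# On update it maps EFortCollectedState to the New or Complete badge and to the encounter text and color, and prints a debug state string.
# Display plays the Intro animation, which drives material parameters, scales and fades, with a sequencer sound event.
# ResetAnims stops animations and plays Intro in reverse.
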WidgetBlueprint /Game/Athena/HUD/Collections/BP_AthenaCollectionItemToast_Character
animations [Intro {
  binding_count=12
  length_seconds=2.8007
  object=&self:Intro
  sequence {
   bindings [
    Image_star {
     id=#1
     possessed_class=Image
     tracks [{
       type=MovieScene2DTransformTrack
       sections [{
         type=MovieScene2DTransformSection
         channels [
          Scale.X {type=MovieSceneScriptingFloatChannel, keys [[27,1],[39,0.5]]}
          Scale.Y {type=MovieSceneScriptingFloatChannel, keys [[27,1],[39,0.5]]}
         ]
        }]
      }]
    }
    LazyImage_ItemImage {
     id=#2
     possessed_class=FortLazyImage
     tracks [
      {
       type=MovieSceneWidgetMaterialTrack
       sections [{
         type=MovieSceneParameterSection
         end_frame=24
         start_frame=0
         channels [Whiteout {type=MovieSceneScriptingFloatChannel, keys [[16,1],[18,0]]}]
        }]
      }
      {
       type=MovieScene2DTransformTrack
       sections [{
         type=MovieScene2DTransformSection
         channels [
          Translation.X {type=MovieSceneScriptingFloatChannel}
          Translation.Y {type=MovieSceneScriptingFloatChannel}
          Angle {type=MovieSceneScriptingFloatChannel}
          Scale.X {type=MovieSceneScriptingFloatChannel, keys [[0,0.75],[16.035,1],[17.005,1.4],[18.7517,1]]}
          Scale.Y {type=MovieSceneScriptingFloatChannel, keys [[0,0.75],[16.035,1],[17.005,1.4],[18.7517,1]]}
          Shear.X {type=MovieSceneScriptingFloatChannel}
          Shear.Y {type=MovieSceneScriptingFloatChannel}
         ]
        }]
      }
     ]
    }
    Image_textglow {
     id=#3
     possessed_class=Image
     tracks [{
       type=MovieSceneFloatTrack
       display_name=Render Opacity
       sections [{type=MovieSceneFloatSection, channels [None {type=MovieSceneScriptingFloatChannel, keys [[27,0],[31,1]]}]}]
      }]
    }
    Image_Swoosh {
     id=#4
     possessed_class=Image
     tracks [
      {
       type=MovieSceneWidgetMaterialTrack
       sections [{
         type=MovieSceneParameterSection
         channels [
          ManualTime {type=MovieSceneScriptingFloatChannel, keys [[0,0],[24,24]]}
          "CrescentOffset (V2)R" {type=MovieSceneScriptingFloatChannel, keys [[0.0697,0],[10.76,0.05],[24.0003,0]]}
          "CrescentOffset (V2)G" {type=MovieSceneScriptingFloatChannel, keys [[0.0697,0],[10.76,0.05],[24.0003,0]]}
          "CrescentOffset (V2)B" {type=MovieSceneScriptingFloatChannel}
          "CrescentOffset (V2)A" {type=MovieSceneScriptingFloatChannel}
         ]
        }]
      }
      {
       type=MovieSceneFloatTrack
       display_name=Render Opacity
       sections [{type=MovieSceneFloatSection, channels [None {type=MovieSceneScriptingFloatChannel, keys [[0,1],[24,0]]}]}]
      }
      {
       type=MovieScene2DTransformTrack
       sections [{
         type=MovieScene2DTransformSection
         channels [
          Scale.X {type=MovieSceneScriptingFloatChannel, keys [[0,1],[15,0.78]]}
          Scale.Y {type=MovieSceneScriptingFloatChannel, keys [[0,1],[15,0.78]]}
         ]
        }]
      }
     ]
    }
    Text_FirstCatch {
     id=#5
     possessed_class=CommonTextBlock
     tracks [{
       type=MovieSceneWidgetMaterialTrack
       display_name=Font.FontMaterial
       sections [{
         type=MovieSceneParameterSection
         end_frame=56
         start_frame=0
         channels [TimeOffset {type=MovieSceneScriptingFloatChannel, keys [[29,0.5071],[29.0003,0.5],[56.013,-1.3581]]}]
        }]
      }]
    }
    Overlay_TopContainer {
     id=#6
     possessed_class=Overlay
     tracks [{
       type=MovieSceneFloatTrack
       display_name=Render Opacity
       sections [{type=MovieSceneFloatSection, channels [None {type=MovieSceneScriptingFloatChannel, keys [[0,0],[2,1]]}]}]
      }]
    }
    HorizontalBox_94 {
     id=#7
     possessed_class=HorizontalBox
     tracks [
      {
       type=MovieSceneFloatTrack
       display_name=Render Opacity
       sections [{type=MovieSceneFloatSection, channels [None {type=MovieSceneScriptingFloatChannel, keys [[29,0],[29.0003,1],[29.505,1]]}]}]
      }
      {
       type=MovieScene2DTransformTrack
       sections [{
         type=MovieScene2DTransformSection
         channels [
          Scale.X {type=MovieSceneScriptingFloatChannel, keys [[29,1],[31,1.25],[31.016,1.5],[32.6573,1]]}
          Scale.Y {type=MovieSceneScriptingFloatChannel, keys [[29,1],[31,1.25],[31.016,1.5],[32.6573,1]]}
         ]
        }]
      }
     ]
    }
    Overlay_Background {
     id=#8
     possessed_class=Overlay
     tracks [{
       type=MovieSceneFloatTrack
       display_name=Render Opacity
       sections [{type=MovieSceneFloatSection, channels [None {type=MovieSceneScriptingFloatChannel, keys [[20,0],[24.0003,1]]}]}]
      }]
    }
    Border_ItemDetails {
     id=#9
     possessed_class=CommonBorder
     tracks [
      {
       type=MovieSceneByteTrack
       display_name=Clipping
       sections [{type=MovieSceneByteSection, channels [None {type=MovieSceneScriptingByteChannel, keys [[20,1],[24,0]]}]}]
      }
      {
       type=MovieSceneFloatVectorTrack
       display_name=Desired Size Scale
       sections [{
         type=MovieSceneFloatVectorSection
         channels [
          Vector.X {type=MovieSceneScriptingFloatChannel, keys [[20,0],[24,1]]}
          Vector.Y {type=MovieSceneScriptingFloatChannel}
         ]
        }]
      }
     ]
    }
    Image_Sparks {
     id=#10
     possessed_class=Image
     tracks [
      {
       type=MovieSceneWidgetMaterialTrack
       sections [{
         type=MovieSceneParameterSection
         end_frame=20
         start_frame=0
         channels [ManualTime {type=MovieSceneScriptingFloatChannel, keys [[0,-2.129],[20,2.36]]}]
        }]
      }
      {
       type=MovieSceneFloatTrack
       display_name=Render Opacity
       sections [{type=MovieSceneFloatSection, channels [None {type=MovieSceneScriptingFloatChannel, keys [[13,1],[20,0]]}]}]
      }
     ]
    }
    Overlay_NewType {
     id=#11
     possessed_class=Overlay
     tracks [
      {
       type=MovieSceneFloatTrack
       display_name=Render Opacity
       sections [{type=MovieSceneFloatSection, channels [None {type=MovieSceneScriptingFloatChannel, keys [[27,0],[30.758,1]]}]}]
      }
      {
       type=MovieScene2DTransformTrack
       sections [{
         type=MovieScene2DTransformSection
         channels [
          Scale.X {type=MovieSceneScriptingFloatChannel, const [1,3]}
          Scale.Y {type=MovieSceneScriptingFloatChannel, const [1,3]}
         ]
        }]
      }
     ]
    }
    Overlay_Complete {
     id=#12
     possessed_class=Overlay
     tracks [
      {
       type=MovieSceneFloatTrack
       display_name=Render Opacity
       sections [{type=MovieSceneFloatSection, channels [None {type=MovieSceneScriptingFloatChannel, keys [[27,0],[30.758,1]]}]}]
      }
      {
       type=MovieScene2DTransformTrack
       sections [{
         type=MovieScene2DTransformSection
         channels [
          Scale.X {type=MovieSceneScriptingFloatChannel, const [1,3]}
          Scale.Y {type=MovieSceneScriptingFloatChannel, const [1,3]}
         ]
        }]
      }
     ]
    }
   ]
   tracks [
    {
     type=MovieSceneEventTrack
     sections [{
       type=MovieSceneEventTriggerSection
       channels [None {type=MovieSceneScriptingEventChannel, const ["<Struct 'MovieSceneEvent' (0x4679f02e380) {}>", 1]}]
      }]
    }
    {type=MovieSceneAudioTrack, display_name=Audio, sections []}
   ]
  }
 }]
graphs [
 EventGraph {nodes [
   ResetAnims() {is=title K2Node_CustomEvent}
   > StopAllAnimations(self:User Widget Object Reference)
   n1 = Get Intro()
   > PlayAnimation(InAnimation=@n1, StartAtTime=GetEndTime(self=@n1), PlayMode=Reverse)
   Event On Display Toast() {is=title K2Node_Event}
   n2 = Get Intro()
   > PlayAnimation(InAnimation=@n2)
   > RetriggerableDelay(Duration=float + float(A=GetEndTime(self=@n2), B=2) {is=title K2Node_CommutativeAssociativeBinaryOperator})
   n3 = Event On Toast Updated!() -> (OutputDelegate, then, CollectedState:EFortCollectedState Enum=Unknown, NumCollected:int=0) {is=title K2Node_Event}
   > n4 = Branch(Condition=Is Design Time!(self:Athena Collection Item Toast Object Reference) -> (ReturnValue:bool))
   ResetAnims(execute=@n4.else) {is=title K2Node_CallFunction}
   n5 = "Equal (Enum)"(A:EFortCollectedState Enum=@n3.CollectedState, B:EFortCollectedState Enum=NewlyCollected)
   n6 = "Equal (Enum)"(A=@n3.CollectedState, B=NewLocation)
   > SetVisibility(
    self=Get Overlay_NewType() -> (Overlay_NewType:Overlay Object Reference)
    InVisibility=Select(Option 0=Collapsed, Option 1=SelfHitTestInvisible, Index=OR Boolean(A=@n5, B=@n6))
   )
   n7 = "Equal (Enum)"(A=@n3.CollectedState, B=NewlyCompleted)
   n8 = "Equal (Enum)"(A=@n3.CollectedState, B=Complete)
   > SetVisibility(
    self=Get Overlay_Complete() -> (Overlay_Complete:Overlay Object Reference)
    InVisibility=Select(Option 0=Collapsed, Option 1=SelfHitTestInvisible, Index=OR Boolean(A=@n7, B=@n8))
   )
   n9 = Get Entry Complete() -> (Entry Complete:Text)
   n10 = Get New Encounter() -> (New Encounter:Text)
   n11 = Get First Encounter() -> (First Encounter:Text)
   n12 = Get Text_FirstCatch() -> (Text_FirstCatch:Common Text Object Reference)
   > "SetText (Text)"(
    self=@n12
    InText=Select!(
     Unknown:Text=@n11
     New:Text=@n11
     Known:Text=@n11
     NewlyCollected:Text=@n11
     Collected:Text=@n11
     NewBest:Text=@n10
     NewRecord:Text=@n10
     NewLocation:Text=@n10
     NewlyCompleted:Text=@n9
     Complete:Text=@n9
     Index:EFortCollectedState Enum=@n3.CollectedState
    ) -> (ReturnValue:Text)
   )
   n13 = Get Color Complete() -> (Color Complete:Slate Color Structure)
   n14 = Get Color NotComplete() -> (Color NotComplete:Slate Color Structure)
   > SetColorAndOpacity(
    self=@n12
    InColorAndOpacity=Select(
     Unknown:Slate Color Structure=@n14
     New:Slate Color Structure=@n14
     Known:Slate Color Structure=@n14
     NewlyCollected:Slate Color Structure=@n14
     Collected:Slate Color Structure=@n14
     NewBest:Slate Color Structure=@n14
     NewRecord:Slate Color Structure=@n14
     NewLocation:Slate Color Structure=@n14
     NewlyCompleted:Slate Color Structure=@n13
     Complete:Slate Color Structure=@n13
     Index=@n3.CollectedState
    ) -> (ReturnValue:Slate Color Structure)
   )
   > PrintString!(
    execute
    InString=Append(A="NPC toast state : ", B=Enum to String!(Enumerator:EFortCollectedState Enum=@n3.CollectedState) -> (ReturnValue:String))
    bPrintToScreen
    bPrintToLog
    TextColor=(R=1,G=0.8418,B=0,A=1)
    Duration=5
    Key
   ) -> (then)
   Event On Toast Reset() {is=title K2Node_Event}
   > SetVisibility(self=[Get Overlay_NewType(), Get Overlay_Complete()], InVisibility=Collapsed)
  ]}
 Sequencer Events {nodes [
   Play Audio() {is=title K2Node_CustomEvent}
   > PlaySound2D(Sound=/Game/Sounds/Fort_UI_Sounds/S15/NPC_Toast_Collected_Cue)
  ]}
]
variables [
 Color Complete {type=SlateColor, default {colorUseRule=UseColor_Specified, specifiedColor=<rgba 1 0.906 0 1>}}
 Color NotComplete {type=SlateColor, default {colorUseRule=UseColor_Specified, specifiedColor=<rgba 0 0.5612 1 1>}}
 Entry Complete {type=string, default="Entry Complete!"}
 First Encounter {type=string, default=First Encounter}
 New Encounter {type=string, default=New Encounter}
]
widgets {
 info {
  inheritedWidgetCount=6
  parentClass=/Script/FortniteUI.AthenaCollectionItemToastCharacter
  rootWidgetClass=&SizeBox
  widgetCount=57
 }
 widgets [
  SizeBox_52 {widgetClassPath=&SizeBox, widget=SizeBox_52, properties {bOverride_HeightOverride=true, heightOverride=128}}
  Overlay_64 {widgetClassPath=&Overlay, parent=SizeBox_52, widget=Overlay_64, slot {class=SizeBoxSlot}}
  Overlay_TopContainer {
   widgetClassPath=&Overlay
   parent=Overlay_64
   widget=Overlay_TopContainer
   slot {class=OverlaySlot, properties {horizontalAlignment=HAlign_Center, verticalAlignment=VAlign_Center}}
  }
  Overlay_Background {
   widgetClassPath=&Overlay
   parent=Overlay_TopContainer
   widget=Overlay_Background
   properties {visibility=Visible}
   slot {class=OverlaySlot, properties {horizontalAlignment=HAlign_Fill, verticalAlignment=VAlign_Fill, padding=<ltrb 64 20 0 0>}}
  }
  SizeBox_0 {
   widgetClassPath=&SizeBox
   parent=Overlay_Background
   widget=SizeBox_0
   properties {bOverride_HeightOverride=true, heightOverride=73, renderTransform {angle=0, scale=<1 1>, shear=<0 0>, translation=<0 -9>}}
   slot {class=OverlaySlot, properties {horizontalAlignment=HAlign_Fill, verticalAlignment=VAlign_Center}}
  }
  Border_Background {
   widgetClassPath=&CommonBorder
   parent=SizeBox_0
   widget=Border_Background
   properties {
    background {
     bIsDynamicallyLoaded=false
     drawAs=Box
     imageType=NoImage
     mirroring=NoMirror
     resourceName=None
     resourceObject=/Game/Athena/HUD/Collections/Materials/M_UI_BackgroundWShadow_Inst
     tiling=NoTile
     imageSize=<128 128>
     margin=<ltrb 0.2 0 0.25 0>
     outlineSettings {
      bUseBrushTransparency=false
      roundingType=HalfHeightRadius
      width=0
      color {colorUseRule=UseColor_Specified, specifiedColor=<rgba 0 0 0 0>}
      cornerRadii=<0 0 0 0>
     }
     tintColor {colorUseRule=UseColor_Specified, specifiedColor=<rgba 1 1 1 1>}
     uVRegion {bIsValid=false, max=<0 0>, min=<0 0>}
    }
    padding=<ltrb 0 0 0 0>
   }
   slot {class=SizeBoxSlot, properties {padding=<ltrb 0 0 -40 -10>}}
  }
  Button_MobileClickCatcher {
   widgetClassPath=&Button
   bInherited=true
   parent=Border_Background
   widget=Button_MobileClickCatcher
   properties {
    isFocusable=false
    backgroundColor=<rgba 1 1 1 0>
    colorAndOpacity=<rgba 1 1 1 0>
    renderTransform {angle=0, scale=<1 1.6>, shear=<0 0>, translation=<0 0>}
   }
   slot {class=BorderSlot, properties {padding=<ltrb 0 0 0 0>}}
   bIsVariable=~
  }
  HorizontalBoxHotkeyPrompt {
   widgetClassPath=&HorizontalBox
   parent=Overlay_Background
   widget=HorizontalBoxHotkeyPrompt
   properties {visibility=Visible, renderTransform {angle=0, scale=<1 1>, shear=<0 0>, translation=<0 15>}}
   slot {class=OverlaySlot, properties {horizontalAlignment=HAlign_Right, verticalAlignment=VAlign_Bottom}}
  }
  SizeBox_Hotkey {
   widgetClassPath=&SizeBox
   parent=HorizontalBoxHotkeyPrompt
   widget=SizeBox_Hotkey
   properties {bOverride_HeightOverride=true, bOverride_WidthOverride=true, heightOverride=48, widthOverride=48}
   slot {class=HorizontalBoxSlot, properties {horizontalAlignment=HAlign_Right, verticalAlignment=VAlign_Top}}
  }
  ScaleBox_0 {widgetClassPath=&ScaleBox, parent=SizeBox_Hotkey, widget=ScaleBox_0, slot {class=SizeBoxSlot}}
  Overlay_1 {widgetClassPath=&Overlay, parent=ScaleBox_0, widget=Overlay_1, slot {class=ScaleBoxSlot}}
  Gamepad {
   widgetClassPath=&CommonVisibilityWidgetLegacy
   parent=Overlay_1
   widget=Gamepad
   properties {bShowForMouseAndKeyboard=false, bShowForTouch=false, padding=<ltrb 0 0 0 0>}
   slot {class=OverlaySlot}
  }
  KeybindWidget {
   widgetClassPath=/FortUILibrary/Input/KeybindWidget
   parent=Gamepad
   widget=KeybindWidget
   properties {boundAction=GamepadToggleFullScreenMap, forcedHoldKeybindStatus=ForcedHold}
   slot {class=BorderSlot, properties {padding=<ltrb 0 0 0 0>}}
  }
  MouseKeyboard {
   widgetClassPath=&CommonVisibilityWidgetLegacy
   parent=Overlay_1
   widget=MouseKeyboard
   properties {
    bShowForAndroid=false
    bShowForGamepad=false
    bShowForIOS=false
    bShowForPS4=false
    bShowForTouch=false
    bShowForXBox=false
    padding=<ltrb 0 0 0 0>
   }
   slot {class=OverlaySlot, properties {horizontalAlignment=HAlign_Fill, verticalAlignment=VAlign_Center}}
  }
  KeybindWidget_1 {
   widgetClassPath=&KeybindWidget
   parent=MouseKeyboard
   widget=KeybindWidget_1
   properties {boundAction=ToggleFullScreenMap, forcedHoldKeybindStatus=ForcedHold}
   slot {class=BorderSlot, properties {padding=<ltrb 0 0 0 0>}}
  }
  Visibility_MobileCollectionText {
   widgetClassPath=&CommonVisibilityWidgetLegacy
   parent=HorizontalBoxHotkeyPrompt
   widget=Visibility_MobileCollectionText
   properties {
    bShowForErebus=false
    bShowForGamepad=false
    bShowForMac=false
    bShowForMouseAndKeyboard=false
    bShowForPC=false
    bShowForPS4=false
    bShowForSwitch2=false
    bShowForXBox=false
   }
   slot {class=HorizontalBoxSlot, properties {horizontalAlignment=HAlign_Right, verticalAlignment=VAlign_Center}}
  }
  Switcher_MobileClueText {
   widgetClassPath=&CommonWidgetSwitcherLegacy
   bIsVariable=true
   parent=Visibility_MobileCollectionText
   widget=Switcher_MobileClueText
   properties {transitionDuration=0}
   slot {class=BorderSlot}
  }
  Text_MobileCollectionPrompt {
   widgetClassPath=&CommonTextBlock
   bIsVariable=true
   parent=Switcher_MobileClueText
   widget=Text_MobileCollectionPrompt
   properties {
    text=TAP TO OPEN COLLECTION
    textTransformPolicy=ToUpper
    font {
     bForceMonospaced=false
     bMaterialIsStencil=false
     fontMaterial=null
     fontObject=&Burbank
     letterSpacing=0
     monospacedWidth=1
     size=12
     skewAmount=0
     typefaceFontName=Burbank-PlayerPower
     outlineSettings {
      bApplyOutlineToDropShadows=false
      bMiteredCorners=false
      bSeparateFillAlpha=false
      outlineMaterial=null
      outlineSize=0
      outlineColor=<rgba 0 0 0 1>
     }
    }
    renderTransform {angle=0, scale=<1 1>, shear=<-12 0>, translation=<0 0>}
    shadowOffset=<0 0>
   }
   slot {class=WidgetSwitcherSlot}
  }
  Overlay_MobileClue {
   widgetClassPath=&Overlay
   parent=Switcher_MobileClueText
   widget=Overlay_MobileClue
   slot {class=WidgetSwitcherSlot, properties {verticalAlignment=VAlign_Center, padding=<ltrb 8 0 16 0>}}
  }
  Text_NewClue {
   widgetClassPath=&CommonTextBlock
   bIsVariable=true
   parent=Overlay_MobileClue
   widget=Text_NewClue
   properties {
    minDesiredWidth=0.5
    text="Tap for New Clue!"
    textTransformPolicy=ToUpper
    colorAndOpacity {colorUseRule=UseColor_Specified, specifiedColor=<rgba 1 0.6445 0 1>}
    font {
     bForceMonospaced=false
     bMaterialIsStencil=false
     fontMaterial=null
     fontObject=&Burbank
     letterSpacing=0
     monospacedWidth=1
     size=12
     skewAmount=0
     typefaceFontName=Burbank-PlayerPower
     outlineSettings {
      bApplyOutlineToDropShadows=false
      bMiteredCorners=false
      bSeparateFillAlpha=false
      outlineMaterial=&M-FontShadow
      outlineSize=3
      outlineColor=<rgba 0 0 0 0.8>
     }
    }
    renderTransform {angle=0, scale=<1 1>, shear=<-8 0>, translation=<0 0>}
    shadowOffset=<2 2>
   }
   slot {class=OverlaySlot, properties {horizontalAlignment=HAlign_Center, verticalAlignment=VAlign_Center, padding=<ltrb 0 0 -2 -2>}}
  }
  Image_glow {
   widgetClassPath=&Image
   parent=Overlay_MobileClue
   widget=Image_glow
   properties {
    brush {
     bIsDynamicallyLoaded=false
     drawAs=Image
     imageType=NoImage
     mirroring=NoMirror
     resourceName=None
     resourceObject=/Game/Athena/HUD/LTM/Wax/Materials/M_UI_TextGlow_Pulse_Add
     tiling=NoTile
     imageSize=<32 32>
     margin=<ltrb 0 0 0 0>
     outlineSettings {
      bUseBrushTransparency=false
      roundingType=HalfHeightRadius
      width=0
      color {colorUseRule=UseColor_Specified, specifiedColor=<rgba 0 0 0 0>}
      cornerRadii=<0 0 0 0>
     }
     tintColor {colorUseRule=UseColor_Specified, specifiedColor=<rgba 1 1 1 1>}
     uVRegion {bIsValid=false, max=<0 0>, min=<0 0>}
    }
    colorAndOpacity=<rgba 0.2159 0.0467 0 1>
   }
   slot {class=OverlaySlot, properties {horizontalAlignment=HAlign_Fill, verticalAlignment=VAlign_Fill, padding=<ltrb -30 0 -30 0>}}
  }
  Visibility_CollectionText {
   widgetClassPath=&CommonVisibilityWidgetLegacy
   parent=HorizontalBoxHotkeyPrompt
   widget=Visibility_CollectionText
   properties {bShowForAndroid=false, bShowForIOS=false, bShowForTouch=false}
   slot {class=HorizontalBoxSlot, properties {horizontalAlignment=HAlign_Right, verticalAlignment=VAlign_Center, padding=<ltrb 8 0 0 0>}}
  }
  Switcher_ClueText {
   widgetClassPath=&CommonWidgetSwitcherLegacy
   bIsVariable=true
   parent=Visibility_CollectionText
   widget=Switcher_ClueText
   properties {transitionDuration=0}
   slot {class=BorderSlot}
  }
  Text_CollectionPrompt {
   widgetClassPath=&CommonTextBlock
   bIsVariable=true
   parent=Switcher_ClueText
   widget=Text_CollectionPrompt
   properties {
    text=Collection
    textTransformPolicy=ToUpper
    font {
     bForceMonospaced=false
     bMaterialIsStencil=false
     fontMaterial=null
     fontObject=&Burbank
     letterSpacing=0
     monospacedWidth=1
     size=12
     skewAmount=0
     typefaceFontName=Burbank-PlayerPower
     outlineSettings {
      bApplyOutlineToDropShadows=false
      bMiteredCorners=false
      bSeparateFillAlpha=false
      outlineMaterial=null
      outlineSize=0
      outlineColor=<rgba 0 0 0 1>
     }
    }
    renderTransform {angle=0, scale=<1 1>, shear=<-8 0>, translation=<0 0>}
    shadowOffset=<0 0>
   }
   slot {class=WidgetSwitcherSlot}
  }
  Overlay_Clue {
   widgetClassPath=&Overlay
   parent=Switcher_ClueText
   widget=Overlay_Clue
   slot {class=WidgetSwitcherSlot, properties {verticalAlignment=VAlign_Center, padding=<ltrb 8 0 16 0>}}
  }
  Text_NewClue_1 {
   widgetClassPath=&CommonTextBlock
   bIsVariable=true
   parent=Overlay_Clue
   widget=Text_NewClue_1
   properties {
    minDesiredWidth=0.5
    text="New Clue!"
    textTransformPolicy=ToUpper
    colorAndOpacity {colorUseRule=UseColor_Specified, specifiedColor=<rgba 1 0.6445 0 1>}
    font {
     bForceMonospaced=false
     bMaterialIsStencil=false
     fontMaterial=null
     fontObject=&Burbank
     letterSpacing=0
     monospacedWidth=1
     size=12
     skewAmount=0
     typefaceFontName=Burbank-PlayerPower
     outlineSettings {
      bApplyOutlineToDropShadows=false
      bMiteredCorners=false
      bSeparateFillAlpha=false
      outlineMaterial=&M-FontShadow
      outlineSize=3
      outlineColor=<rgba 0 0 0 0.8>
     }
    }
    renderTransform {angle=0, scale=<1 1>, shear=<-8 0>, translation=<0 0>}
    shadowOffset=<2 2>
   }
   slot {class=OverlaySlot, properties {horizontalAlignment=HAlign_Center, verticalAlignment=VAlign_Center, padding=<ltrb 0 0 -2 -2>}}
  }
  Image_glow_1 {
   widgetClassPath=&Image
   parent=Overlay_Clue
   widget=Image_glow_1
   properties {
    brush {
     bIsDynamicallyLoaded=false
     drawAs=Image
     imageType=NoImage
     mirroring=NoMirror
     resourceName=None
     resourceObject=&M_UI_TextGlow_Pulse_Add
     tiling=NoTile
     imageSize=<32 32>
     margin=<ltrb 0 0 0 0>
     outlineSettings {
      bUseBrushTransparency=false
      roundingType=HalfHeightRadius
      width=0
      color {colorUseRule=UseColor_Specified, specifiedColor=<rgba 0 0 0 0>}
      cornerRadii=<0 0 0 0>
     }
     tintColor {colorUseRule=UseColor_Specified, specifiedColor=<rgba 1 1 1 1>}
     uVRegion {bIsValid=false, max=<0 0>, min=<0 0>}
    }
    colorAndOpacity=<rgba 0.2159 0.0467 0 1>
   }
   slot {class=OverlaySlot, properties {horizontalAlignment=HAlign_Fill, verticalAlignment=VAlign_Fill, padding=<ltrb -30 0 -30 0>}}
  }
  HorizontalBox_ImageAndDetails {
   widgetClassPath=&HorizontalBox
   parent=Overlay_TopContainer
   widget=HorizontalBox_ImageAndDetails
   slot {class=OverlaySlot, properties {horizontalAlignment=HAlign_Fill, verticalAlignment=VAlign_Fill}}
  }
  Overlay_ItemImage {
   widgetClassPath=&Overlay
   parent=HorizontalBox_ImageAndDetails
   widget=Overlay_ItemImage
   slot {class=HorizontalBoxSlot, properties {horizontalAlignment=HAlign_Center, verticalAlignment=VAlign_Center}}
  }
  Image_textglow_1 {
   widgetClassPath=&Image
   parent=Overlay_ItemImage
   widget=Image_textglow_1
   properties {
    visibility=SelfHitTestInvisible
    brush {
     bIsDynamicallyLoaded=false
     drawAs=Image
     imageType=NoImage
     mirroring=NoMirror
     resourceName=None
     resourceObject=/Game/Athena/HUD/LTM/Wax/Materials/M_UI_TextGlow_Add
     tiling=NoTile
     imageSize=<105.9261 123.5972>
     margin=<ltrb 0 0 0 0>
     outlineSettings {
      bUseBrushTransparency=false
      roundingType=HalfHeightRadius
      width=0
      color {colorUseRule=UseColor_Specified, specifiedColor=<rgba 0 0 0 0>}
      cornerRadii=<0 0 0 0>
     }
     tintColor {colorUseRule=UseColor_Specified, specifiedColor=<rgba 0.0037 0.3613 1 1>}
     uVRegion {bIsValid=false, max=<0 0>, min=<0 0>}
    }
   }
   slot {class=OverlaySlot, properties {horizontalAlignment=HAlign_Center, verticalAlignment=VAlign_Center, padding=<ltrb -75 -75 -75 -75>}}
  }
  LazyImage_ItemImage {
   widgetClassPath=&FortLazyImage
   bInherited=true
   parent=Overlay_ItemImage
   widget=LazyImage_ItemImage
   properties {
    materialTextureParamName=Texture
    brush {
     bIsDynamicallyLoaded=false
     drawAs=Image
     imageType=NoImage
     mirroring=NoMirror
     resourceName=None
     resourceObject=/Game/Athena/HUD/Collections/Materials/M_UI_FishToast_Character
     tiling=NoTile
     imageSize=<128 128>
     margin=<ltrb 0 0 0 0>
     outlineSettings {
      bUseBrushTransparency=false
      roundingType=HalfHeightRadius
      width=0
      color {colorUseRule=UseColor_Specified, specifiedColor=<rgba 0 0 0 0>}
      cornerRadii=<0 0 0 0>
     }
     tintColor {colorUseRule=UseColor_Specified, specifiedColor=<rgba 1 1 1 1>}
     uVRegion {bIsValid=false, max=<0 0>, min=<0 0>}
    }
    shadowColorAndOpacity=<rgba 0.0052 0.0052 0.0056 0.5>
   }
   slot {class=OverlaySlot, properties {horizontalAlignment=HAlign_Center, verticalAlignment=VAlign_Center}}
  }
  Image_Sparks {
   widgetClassPath=&Image
   parent=Overlay_ItemImage
   widget=Image_Sparks
   properties {
    renderOpacity=0
    visibility=SelfHitTestInvisible
    brush {
     bIsDynamicallyLoaded=false
     drawAs=Image
     imageType=NoImage
     mirroring=NoMirror
     resourceName=None
     resourceObject=/Game/Athena/UI/InGame/Interrogation/Materials/M_UI_Starburst_Inst3
     tiling=NoTile
     imageSize=<32 32>
     margin=<ltrb 0 0 0 0>
     outlineSettings {
      bUseBrushTransparency=false
      roundingType=HalfHeightRadius
      width=0
      color {colorUseRule=UseColor_Specified, specifiedColor=<rgba 0 0 0 0>}
      cornerRadii=<0 0 0 0>
     }
     tintColor {colorUseRule=UseColor_Specified, specifiedColor=<rgba 1 1 1 1>}
     uVRegion {bIsValid=false, max=<0 0>, min=<0 0>}
    }
   }
   slot {class=OverlaySlot, properties {horizontalAlignment=HAlign_Fill, verticalAlignment=VAlign_Fill, padding=<ltrb -100 -100 -100 -100>}}
  }
  Image_Swoosh {
   widgetClassPath=&Image
   parent=Overlay_ItemImage
   widget=Image_Swoosh
   properties {
    renderOpacity=0
    visibility=SelfHitTestInvisible
    brush {
     bIsDynamicallyLoaded=false
     drawAs=Image
     imageType=NoImage
     mirroring=NoMirror
     resourceName=None
     resourceObject=/Game/Athena/HUD/Collections/Materials/M_UI_PerspSwoosh_Inst
     tiling=NoTile
     imageSize=<200 200>
     margin=<ltrb 0 0 0 0>
     outlineSettings {
      bUseBrushTransparency=false
      roundingType=HalfHeightRadius
      width=0
      color {colorUseRule=UseColor_Specified, specifiedColor=<rgba 0 0 0 0>}
      cornerRadii=<0 0 0 0>
     }
     tintColor {colorUseRule=UseColor_Specified, specifiedColor=<rgba 1 1 1 1>}
     uVRegion {bIsValid=false, max=<0 0>, min=<0 0>}
    }
   }
   slot {class=OverlaySlot, properties {horizontalAlignment=HAlign_Center, verticalAlignment=VAlign_Center, padding=<ltrb -200 -200 -200 -200>}}
  }
  Overlay_Stars {
   widgetClassPath=&Overlay
   parent=Overlay_ItemImage
   widget=Overlay_Stars
   properties {renderOpacity=0, renderTransform {angle=8, scale=<1 1>, shear=<0 0>, translation=<-23 -36>}}
   slot {class=OverlaySlot, properties {horizontalAlignment=HAlign_Right, verticalAlignment=VAlign_Bottom, padding=<ltrb -55 -55 -55 -55>}}
  }
  Image_StarRating {
   widgetClassPath=&FortLazyImage
   bIsVariable=true
   parent=Overlay_Stars
   widget=Image_StarRating
   properties {
    bApplyDropShadow=true
    brush {
     bIsDynamicallyLoaded=false
     drawAs=Image
     imageType=NoImage
     mirroring=NoMirror
     resourceName=None
     resourceObject=/Game/Athena/Collections/Art/M_Ui_Collection_HighScore_IconBig_Toast
     tiling=NoTile
     imageSize=<116 116>
     margin=<ltrb 0 0 0 0>
     outlineSettings {
      bUseBrushTransparency=false
      roundingType=HalfHeightRadius
      width=0
      color {colorUseRule=UseColor_Specified, specifiedColor=<rgba 0 0 0 0>}
      cornerRadii=<0 0 0 0>
     }
     tintColor {colorUseRule=UseColor_Specified, specifiedColor=<rgba 1 1 1 1>}
     uVRegion {bIsValid=false, max=<0 0>, min=<0 0>}
    }
    shadowColorAndOpacity=<rgba 0 0 0 0.5>
   }
   slot {class=OverlaySlot}
  }
  Image_3starglow {
   widgetClassPath=&Image
   bIsVariable=true
   parent=Overlay_Stars
   widget=Image_3starglow
   properties {
    visibility=HitTestInvisible
    brush {
     bIsDynamicallyLoaded=false
     drawAs=Image
     imageType=NoImage
     mirroring=NoMirror
     resourceName=None
     resourceObject=&M_UI_TextGlow_Pulse_Add
     tiling=NoTile
     imageSize=<82.1221 71.9925>
     margin=<ltrb 0 0 0 0>
     outlineSettings {
      bUseBrushTransparency=false
      roundingType=HalfHeightRadius
      width=0
      color {colorUseRule=UseColor_Specified, specifiedColor=<rgba 0 0 0 0>}
      cornerRadii=<0 0 0 0>
     }
     tintColor {colorUseRule=UseColor_Specified, specifiedColor=<rgba 1 0.3468 0 1>}
     uVRegion {bIsValid=false, max=<0 0>, min=<0 0>}
    }
   }
   slot {class=OverlaySlot, properties {horizontalAlignment=HAlign_Center, verticalAlignment=VAlign_Center, padding=<ltrb -75 -75 -75 -75>}}
  }
  Border_ItemDetails {
   widgetClassPath=&CommonBorder
   parent=HorizontalBox_ImageAndDetails
   widget=Border_ItemDetails
   properties {clipping=ClipToBounds, padding=<ltrb 8 0 8 0>}
   slot {class=HorizontalBoxSlot, properties {verticalAlignment=VAlign_Center}}
  }
  VerticalBox_0 {
   widgetClassPath=&VerticalBox
   parent=Border_ItemDetails
   widget=VerticalBox_0
   properties {renderTransform {angle=0, scale=<1 1>, shear=<-8 0>, translation=<0 0>}}
   slot {class=BorderSlot, properties {padding=<ltrb 8 0 8 0>}}
  }
  CommonBorder_ItemName {
   widgetClassPath=&CommonBorder
   parent=VerticalBox_0
   widget=CommonBorder_ItemName
   properties {clipping=ClipToBounds, verticalAlignment=VAlign_Bottom, padding=<ltrb 0 0 0 0>}
   slot {class=VerticalBoxSlot}
  }
  HorizontalBox_58 {
   widgetClassPath=&HorizontalBox
   parent=CommonBorder_ItemName
   widget=HorizontalBox_58
   slot {class=BorderSlot, properties {verticalAlignment=VAlign_Bottom, padding=<ltrb 0 0 0 0>}}
  }
  CommonTextBlock_277 {
   widgetClassPath=&CommonTextBlock
   parent=HorizontalBox_58
   widget=CommonTextBlock_277
   properties {
    text=No.
    textTransformPolicy=ToUpper
    font {
     bForceMonospaced=false
     bMaterialIsStencil=false
     fontMaterial=null
     fontObject=&Burbank
     letterSpacing=0
     monospacedWidth=1
     size=12
     skewAmount=0
     typefaceFontName=Burbank-PlayerPower
     outlineSettings {
      bApplyOutlineToDropShadows=false
      bMiteredCorners=false
      bSeparateFillAlpha=false
      outlineMaterial=null
      outlineSize=0
      outlineColor=<rgba 0 0 0 1>
     }
    }
    shadowOffset=<0 0>
   }
   slot {class=HorizontalBoxSlot, properties {verticalAlignment=VAlign_Top, padding=<ltrb 0 2 1 0>}}
  }
  Text_ItemNumber {
   widgetClassPath=&CommonTextBlock
   bInherited=true
   bIsVariable=true
   parent=HorizontalBox_58
   widget=Text_ItemNumber
   properties {
    text=00
    textTransformPolicy=ToUpper
    font {
     bForceMonospaced=false
     bMaterialIsStencil=false
     fontMaterial=null
     fontObject=&Burbank
     letterSpacing=0
     monospacedWidth=1
     size=16
     skewAmount=0
     typefaceFontName=Burbank-PlayerPower
     outlineSettings {
      bApplyOutlineToDropShadows=false
      bMiteredCorners=false
      bSeparateFillAlpha=false
      outlineMaterial=null
      outlineSize=0
      outlineColor=<rgba 0 0 0 1>
     }
    }
    shadowOffset=<0 0>
   }
   slot {class=HorizontalBoxSlot, properties {padding=<ltrb 0 0 8 0>}}
  }
  Text_ItemName {
   widgetClassPath=&CommonTextBlock
   bInherited=true
   bIsVariable=true
   parent=HorizontalBox_58
   widget=Text_ItemName
   properties {
    text=NPC NAME
    textTransformPolicy=ToUpper
    font {
     bForceMonospaced=false
     bMaterialIsStencil=false
     fontMaterial=null
     fontObject=&Burbank
     letterSpacing=0
     monospacedWidth=1
     size=16
     skewAmount=0
     typefaceFontName=Burbank-PlayerPower
     outlineSettings {
      bApplyOutlineToDropShadows=false
      bMiteredCorners=false
      bSeparateFillAlpha=false
      outlineMaterial=null
      outlineSize=0
      outlineColor=<rgba 0 0 0 1>
     }
    }
    shadowOffset=<0 0>
   }
   slot {class=HorizontalBoxSlot}
  }
  HorizontalBox_94 {widgetClassPath=&HorizontalBox, parent=VerticalBox_0, widget=HorizontalBox_94, slot {class=VerticalBoxSlot}}
  Text_FirstCatch {
   widgetClassPath=&CommonTextBlock
   bIsVariable=true
   parent=HorizontalBox_94
   widget=Text_FirstCatch
   properties {
    text=First Encounter
    textTransformPolicy=ToUpper
    colorAndOpacity {colorUseRule=UseColor_Specified, specifiedColor=<rgba 0 0.5612 1 1>}
    font {
     bForceMonospaced=false
     bMaterialIsStencil=false
     fontMaterial=/Game/Athena/HUD/Materials/M_UI_OutlineShine_White_Manual
     fontObject=&Burbank
     letterSpacing=0
     monospacedWidth=1
     size=16
     skewAmount=0
     typefaceFontName=Burbank-PlayerPower
     outlineSettings {
      bApplyOutlineToDropShadows=false
      bMiteredCorners=false
      bSeparateFillAlpha=false
      outlineMaterial=null
      outlineSize=0
      outlineColor=<rgba 0 0 0 1>
     }
    }
    renderTransformPivot=<0 0.5>
    shadowOffset=<0 0>
   }
   slot {class=HorizontalBoxSlot}
  }
  RichText_NumLocations {
   widgetClassPath=&CommonRichTextBlock
   bInherited=true
   bIsVariable=true
   parent=HorizontalBox_94
   widget=RichText_NumLocations
   properties {
    bOverrideDefaultStyle=true
    text=0/3
    defaultTextStyleOverride {
     overflowPolicy=Clip
     transformPolicy=None
     colorAndOpacity {colorUseRule=UseColor_Specified, specifiedColor=<rgba 1 1 1 1>}
     font {
      bForceMonospaced=false
      bMaterialIsStencil=false
      fontMaterial=null
      fontObject=&Burbank
      letterSpacing=0
      monospacedWidth=1
      size=14
      skewAmount=0
      typefaceFontName=Burbank-PlayerPower
      outlineSettings {
       bApplyOutlineToDropShadows=false
       bMiteredCorners=false
       bSeparateFillAlpha=false
       outlineMaterial=null
       outlineSize=0
       outlineColor=<rgba 0 0 0 1>
      }
     }
     highlightColor {colorUseRule=UseColor_Specified, specifiedColor=<rgba 0 0 0 1>}
     highlightShape {
      bIsDynamicallyLoaded=false
      drawAs=Image
      imageType=NoImage
      mirroring=NoMirror
      resourceName=None
      resourceObject=null
      tiling=NoTile
      imageSize=<32 32>
      margin=<ltrb 0 0 0 0>
      outlineSettings {
       bUseBrushTransparency=false
       roundingType=HalfHeightRadius
       width=0
       color {colorUseRule=UseColor_Specified, specifiedColor=<rgba 0 0 0 0>}
       cornerRadii=<0 0 0 0>
      }
      tintColor {colorUseRule=UseColor_Specified, specifiedColor=<rgba 1 1 1 1>}
      uVRegion {bIsValid=false, max=<0 0>, min=<0 0>}
     }
     selectedBackgroundColor {colorUseRule=UseColor_Foreground, specifiedColor=<rgba 1 0 1 1>}
     shadowColorAndOpacity=<rgba 0 0 0 1>
     shadowOffset=<0 0>
     strikeBrush {
      bIsDynamicallyLoaded=false
      drawAs=Image
      imageType=NoImage
      mirroring=NoMirror
      resourceName=None
      resourceObject=null
      tiling=NoTile
      imageSize=<32 32>
      margin=<ltrb 0 0 0 0>
      outlineSettings {
       bUseBrushTransparency=false
       roundingType=HalfHeightRadius
       width=0
       color {colorUseRule=UseColor_Specified, specifiedColor=<rgba 0 0 0 0>}
       cornerRadii=<0 0 0 0>
      }
      tintColor {colorUseRule=UseColor_Specified, specifiedColor=<rgba 1 1 1 1>}
      uVRegion {bIsValid=false, max=<0 0>, min=<0 0>}
     }
     underlineBrush {
      bIsDynamicallyLoaded=false
      drawAs=Image
      imageType=NoImage
      mirroring=NoMirror
      resourceName=None
      resourceObject=null
      tiling=NoTile
      imageSize=<32 32>
      margin=<ltrb 0 0 0 0>
      outlineSettings {
       bUseBrushTransparency=false
       roundingType=HalfHeightRadius
       width=0
       color {colorUseRule=UseColor_Specified, specifiedColor=<rgba 0 0 0 0>}
       cornerRadii=<0 0 0 0>
      }
      tintColor {colorUseRule=UseColor_Specified, specifiedColor=<rgba 1 1 1 1>}
      uVRegion {bIsValid=false, max=<0 0>, min=<0 0>}
     }
    }
   }
   slot {class=HorizontalBoxSlot, properties {verticalAlignment=VAlign_Bottom, padding=<ltrb 6 0 0 0>}}
  }
  Overlay_NewType {
   widgetClassPath=&Overlay
   bIsVariable=true
   parent=Overlay_TopContainer
   widget=Overlay_NewType
   properties {visibility=Collapsed, renderTransform {angle=0, scale=<1 1>, shear=<0 0>, translation=<25 0>}}
   slot {class=OverlaySlot, properties {verticalAlignment=VAlign_Bottom}}
  }
  Image_textglow {
   widgetClassPath=&Image
   parent=Overlay_NewType
   widget=Image_textglow
   properties {
    visibility=HitTestInvisible
    brush {
     bIsDynamicallyLoaded=false
     drawAs=Image
     imageType=NoImage
     mirroring=NoMirror
     resourceName=None
     resourceObject=&M_UI_TextGlow_Add
     tiling=NoTile
     imageSize=<82.1221 71.9925>
     margin=<ltrb 0 0 0 0>
     outlineSettings {
      bUseBrushTransparency=false
      roundingType=HalfHeightRadius
      width=0
      color {colorUseRule=UseColor_Specified, specifiedColor=<rgba 0 0 0 0>}
      cornerRadii=<0 0 0 0>
     }
     tintColor {colorUseRule=UseColor_Specified, specifiedColor=<rgba 0 0.1421 1 1>}
     uVRegion {bIsValid=false, max=<0 0>, min=<0 0>}
    }
    renderTransform {angle=0, scale=<1 1>, shear=<0 0>, translation=<1 -12>}
   }
   slot {class=OverlaySlot, properties {horizontalAlignment=HAlign_Center, verticalAlignment=VAlign_Center, padding=<ltrb -75 -75 -75 -75>}}
  }
  Image_star {
   widgetClassPath=&Image
   parent=Overlay_NewType
   widget=Image_star
   properties {
    visibility=HitTestInvisible
    brush {
     bIsDynamicallyLoaded=false
     drawAs=Image
     imageType=NoImage
     mirroring=NoMirror
     resourceName=None
     resourceObject=/Game/UI/InGame/Quest/Materials/M_UI_SparkStar
     tiling=NoTile
     imageSize=<128 128>
     margin=<ltrb 0 0 0 0>
     outlineSettings {
      bUseBrushTransparency=false
      roundingType=HalfHeightRadius
      width=0
      color {colorUseRule=UseColor_Specified, specifiedColor=<rgba 0 0 0 0>}
      cornerRadii=<0 0 0 0>
     }
     tintColor {colorUseRule=UseColor_Specified, specifiedColor=<rgba 1 1 1 1>}
     uVRegion {bIsValid=false, max=<0 0>, min=<0 0>}
    }
   }
   slot {class=OverlaySlot, properties {horizontalAlignment=HAlign_Center, verticalAlignment=VAlign_Center, padding=<ltrb -64 -64 -64 -64>}}
  }
  VisibilitySwitcher_ToastType {
   widgetClassPath=&CommonVisibilitySwitcher
   bIsVariable=true
   parent=Overlay_NewType
   widget=VisibilitySwitcher_ToastType
   slot {class=OverlaySlot}
  }
  Text_New {
   widgetClassPath=&CommonTextBlock
   parent=VisibilitySwitcher_ToastType
   widget=Text_New
   properties {
    justification=Center
    text="New!"
    textTransformPolicy=ToUpper
    font {
     bForceMonospaced=false
     bMaterialIsStencil=false
     fontMaterial=null
     fontObject=&Burbank
     letterSpacing=0
     monospacedWidth=1
     size=20
     skewAmount=0
     typefaceFontName=BurbankRegular-Black
     outlineSettings {
      bApplyOutlineToDropShadows=true
      bMiteredCorners=false
      bSeparateFillAlpha=false
      outlineMaterial=null
      outlineSize=1
      outlineColor=<rgba 0 0.0326 1 1>
     }
    }
    renderTransform {angle=-10, scale=<1 1>, shear=<0 0>, translation=<0 0>}
    shadowColorAndOpacity=<rgba 0 0.0326 1 1>
    shadowOffset=<1.5 1.5>
   }
   slot {
    class=CommonVisibilitySwitcherSlot
    properties {horizontalAlignment=HAlign_Center, verticalAlignment=VAlign_Center, padding=<ltrb -200 0 -200 0>}
   }
  }
  Image_textglow_2 {
   widgetClassPath=&Image
   parent=Overlay_NewType
   widget=Image_textglow_2
   properties {
    visibility=SelfHitTestInvisible
    brush {
     bIsDynamicallyLoaded=false
     drawAs=Image
     imageType=NoImage
     mirroring=NoMirror
     resourceName=None
     resourceObject=&M_UI_TextGlow_Add
     tiling=NoTile
     imageSize=<16 10>
     margin=<ltrb 0 0 0 0>
     outlineSettings {
      bUseBrushTransparency=false
      roundingType=HalfHeightRadius
      width=0
      color {colorUseRule=UseColor_Specified, specifiedColor=<rgba 0 0 0 0>}
      cornerRadii=<0 0 0 0>
     }
     tintColor {colorUseRule=UseColor_Specified, specifiedColor=<rgba 0.0037 0.3613 1 1>}
     uVRegion {bIsValid=false, max=<0 0>, min=<0 0>}
    }
    renderTransform {angle=-7, scale=<1 1>, shear=<0 0>, translation=<0 0>}
   }
   slot {class=OverlaySlot, properties {horizontalAlignment=HAlign_Fill, verticalAlignment=VAlign_Fill, padding=<ltrb -50 5 -50 5>}}
  }
  Overlay_Complete {
   widgetClassPath=&Overlay
   bIsVariable=true
   parent=Overlay_TopContainer
   widget=Overlay_Complete
   properties {visibility=Collapsed, renderTransform {angle=0, scale=<1 1>, shear=<0 0>, translation=<76 6>}}
   slot {class=OverlaySlot, properties {verticalAlignment=VAlign_Bottom}}
  }
  Image_textglow_3 {
   widgetClassPath=&Image
   parent=Overlay_Complete
   widget=Image_textglow_3
   properties {
    visibility=HitTestInvisible
    brush {
     bIsDynamicallyLoaded=false
     drawAs=Image
     imageType=NoImage
     mirroring=NoMirror
     resourceName=None
     resourceObject=/Game/Athena/HUD/LTM/Wax/Materials/M_UI_TextGlow
     tiling=NoTile
     imageSize=<64 64>
     margin=<ltrb 0 0 0 0>
     outlineSettings {
      bUseBrushTransparency=false
      roundingType=HalfHeightRadius
      width=0
      color {colorUseRule=UseColor_Specified, specifiedColor=<rgba 0 0 0 0>}
      cornerRadii=<0 0 0 0>
     }
     tintColor {colorUseRule=UseColor_Specified, specifiedColor=<rgba 0.5104 0.0466 0 0.75>}
     uVRegion {bIsValid=false, max=<0 0>, min=<0 0>}
    }
    renderTransform {angle=0, scale=<1.8 1.8>, shear=<0 0>, translation=<1 0>}
   }
   slot {class=OverlaySlot, properties {horizontalAlignment=HAlign_Center, verticalAlignment=VAlign_Center, padding=<ltrb -75 -75 -75 -75>}}
  }
  Image_79 {
   widgetClassPath=&Image
   bIsVariable=true
   parent=Overlay_Complete
   widget=Image_79
   properties {
    brush {
     bIsDynamicallyLoaded=false
     drawAs=Image
     imageType=NoImage
     mirroring=NoMirror
     resourceName=None
     resourceObject=/Game/Athena/Collections/Art/Character/M_Ui_LocationMaker_64_Inst
     tiling=NoTile
     imageSize=<64 64>
     margin=<ltrb 0 0 0 0>
     outlineSettings {
      bUseBrushTransparency=false
      roundingType=HalfHeightRadius
      width=0
      color {colorUseRule=UseColor_Specified, specifiedColor=<rgba 0 0 0 0>}
      cornerRadii=<0 0 0 0>
     }
     tintColor {colorUseRule=UseColor_Specified, specifiedColor=<rgba 1 0.792 0 1>}
     uVRegion {bIsValid=false, max=<0 0>, min=<0 0>}
    }
    renderTransform {angle=7, scale=<1 1>, shear=<0 0>, translation=<0 0>}
   }
   slot {class=OverlaySlot}
  }
  Image_textglow_4 {
   widgetClassPath=&Image
   parent=Overlay_Complete
   widget=Image_textglow_4
   properties {
    visibility=SelfHitTestInvisible
    brush {
     bIsDynamicallyLoaded=false
     drawAs=Image
     imageType=NoImage
     mirroring=NoMirror
     resourceName=None
     resourceObject=&M_UI_TextGlow_Add
     tiling=NoTile
     imageSize=<16 10>
     margin=<ltrb 0 0 0 0>
     outlineSettings {
      bUseBrushTransparency=false
      roundingType=HalfHeightRadius
      width=0
      color {colorUseRule=UseColor_Specified, specifiedColor=<rgba 0 0 0 0>}
      cornerRadii=<0 0 0 0>
     }
     tintColor {colorUseRule=UseColor_Specified, specifiedColor=<rgba 1 0.906 0 0.0857>}
     uVRegion {bIsValid=false, max=<0 0>, min=<0 0>}
    }
    renderTransform {angle=10, scale=<1 1>, shear=<0 0>, translation=<0 0>}
   }
   slot {class=OverlaySlot, properties {horizontalAlignment=HAlign_Fill, verticalAlignment=VAlign_Fill, padding=<ltrb -50 5 -50 5>}}
  }
  TutorialHighlight {widgetClassPath=&TutorialHighlight}
 ]
}
