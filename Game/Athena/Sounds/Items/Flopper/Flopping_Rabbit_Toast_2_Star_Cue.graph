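# Sound cue for the two-star Flopping Rabbit Toast flopper item, judging by its name and folder.
# A single wave player feeds the output directly, with no randomization or modulation.
# It plays in the impacts sound class with the forged consumable in-world attenuation, a short audible range and a slight volume boost.
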SoundCue /Game/Athena/Sounds/Items/Flopper/Flopping_Rabbit_Toast_2_Star_Cue
root=WavePlayer_1
nodes [{id=WavePlayer_1, properties {soundWaveAssetPtr=&/Flopping_Rabbit_Toast_2_Stars}}]
properties {
 attenuationSettings=/Game/Sounds/Attenuation/Fort_ForgedConsumable_InWorld_Att
 maxDistance=1000
 soundClassObject=&Fort_Impacts
 volumeMultiplier=1.0714
}
