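# Sound cue for the first-catch toast of the Flopping Rabbit flopper item, judging by its name.
# A single wave player plays the Flopping Rabbit first-catch wave with no randomization or mixing.
# The cue raises pitch and volume slightly, routes through the impacts sound class, and uses forged-consumable in-world attenuation for positional falloff.
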SoundCue /Game/Athena/Sounds/Items/Flopper/Flopping_Rabbit_Toast_First_Catch
root=WavePlayer_1
nodes [{id=WavePlayer_1, properties {soundWaveAssetPtr=&/Flopping_Rabbit_First_Catch}}]
properties {
 attenuationSettings=/Game/Sounds/Attenuation/Fort_ForgedConsumable_InWorld_Att
 maxDistance=1000
 pitchMultiplier=1.15
 soundClassObject=&Fort_Impacts
 volumeMultiplier=1.3571
}
