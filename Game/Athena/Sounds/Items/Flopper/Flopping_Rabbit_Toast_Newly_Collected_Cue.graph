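# Sound cue for the toast shown when a Flopping Rabbit flopper is newly collected, judging by its name.
# A single wave player feeds the fishing mini-game toast sound straight to the output, with no randomization or mixing.
# Playback uses the in-world forged consumable attenuation and the Fort_Impacts sound class, with a boosted volume multiplier.
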SoundCue /Game/Athena/Sounds/Items/Flopper/Flopping_Rabbit_Toast_Newly_Collected_Cue
root=WavePlayer_0
nodes [{id=WavePlayer_0, properties {soundWaveAssetPtr=&/Fishing_Mini_Game_Toast_v5}}]
properties {
 attenuationSettings=/Game/Sounds/Attenuation/Fort_ForgedConsumable_InWorld_Att
 maxDistance=1000
 soundClassObject=&Fort_Impacts
 volumeMultiplier=1.3571
}
