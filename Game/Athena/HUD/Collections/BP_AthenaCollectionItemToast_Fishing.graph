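# Fishing collection toast widget that shows a caught fish with first-catch or record state, star rating, relative size and leaderboard info.
# Parent-class events drive switchers, visibility, star and size-meter material parameters, and queue "New Clue" text until a toast runs.
# Display plays NewFish, then LeaderTransition after a delay; sequencer events trigger star animation and state-based sounds.
# ResetAnims rewinds animations by playing them reversed from their end time.
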WidgetBlueprint /Game/Athena/HUD/Collections/BP_AthenaCollectionItemToast_Fishing
animations [
 NewFish {
  binding_count=10
  length_seconds=2.8127
  object=&self:NewFish
  sequence {
   bindings [
    Image_Sparks {
     id=#1
     possessed_class=Image
     tracks [
      {
       type=MovieSceneWidgetMaterialTrack
       sections [{type=MovieSceneParameterSection, channels [ManualTime {type=MovieSceneScriptingFloatChannel, keys [[-0.2563,-2.129],[20.01,2.36]]}]}]
      }
      {
       type=MovieSceneFloatTrack
       display_name=Render Opacity
       sections [{type=MovieSceneFloatSection, channels [None {type=MovieSceneScriptingFloatChannel, keys [[12.7377,1],[19.7877,0]]}]}]
      }
     ]
    }
    LazyImage_ItemImage {
     id=#2
     possessed_class=FortLazyImage
     tracks [
      {
       type=MovieSceneWidgetMaterialTrack
       sections [{type=MovieSceneParameterSection, channels [Whiteout {type=MovieSceneScriptingFloatChannel, keys [[15.7787,1],[18.4953,0]]}]}]
      }
      {
       type=MovieScene2DTransformTrack
       sections [{
         type=MovieScene2DTransformSection
         channels [
          Translation.X {type=MovieSceneScriptingFloatChannel}
          Translation.Y {type=MovieSceneScriptingFloatChannel}
          Angle {type=MovieSceneScriptingFloatChannel}
          Scale.X {type=MovieSceneScriptingFloatChannel, keys [[-0.2563,0.8],[15.7787,1],[16.7487,1.4],[18.4953,1]]}
          Scale.Y {type=MovieSceneScriptingFloatChannel, keys [[-0.2563,0.8],[15.7787,1],[16.7487,1.4],[18.4953,1]]}
          Shear.X {type=MovieSceneScriptingFloatChannel}
          Shear.Y {type=MovieSceneScriptingFloatChannel}
         ]
        }]
      }
     ]
    }
    Border_ItemDetails {
     id=#3
     possessed_class=CommonBorder
     tracks [
      {
       type=MovieSceneFloatVectorTrack
       display_name=Desired Size Scale
       sections [{
         type=MovieSceneFloatVectorSection
         channels [
          Vector.X {type=MovieSceneScriptingFloatChannel, keys [[19.7877,0],[24.0647,1]]}
          Vector.Y {type=MovieSceneScriptingFloatChannel}
         ]
        }]
      }
      {
       type=MovieSceneByteTrack
       display_name=Clipping
       sections [{type=MovieSceneByteSection, channels [None {type=MovieSceneScriptingByteChannel, keys [[19.7877,1],[24.0647,0]]}]}]
      }
     ]
    }
    VerticalBox_1 {
     id=#4
     display_name="LazyImage_ItemImage (10)"
     possessed_class=VerticalBox
     tracks [{
       type=MovieSceneFloatTrack
       display_name=Render Opacity
       sections [{type=MovieSceneFloatSection, channels [None {type=MovieSceneScriptingFloatChannel, keys [[19.7877,0],[24.0647,1]]}]}]
      }]
    }
    Image_Swoosh {
     id=#5
     possessed_class=Image
     tracks [
      {
       type=MovieSceneWidgetMaterialTrack
       sections [{
         type=MovieSceneParameterSection
         channels [
          ManualTime {type=MovieSceneScriptingFloatChannel, keys [[-0.2563,0.01],[23.6743,24.8149]]}
          "CrescentOffset (V2)R" {type=MovieSceneScriptingFloatChannel, keys [[-0.2563,0],[10.434,0.05],[23.6743,0]]}
          "CrescentOffset (V2)G" {type=MovieSceneScriptingFloatChannel, keys [[-0.2563,0],[10.434,0.05],[23.6743,0]]}
          "CrescentOffset (V2)B" {type=MovieSceneScriptingFloatChannel}
          "CrescentOffset (V2)A" {type=MovieSceneScriptingFloatChannel}
         ]
        }]
      }
      {
       type=MovieScene2DTransformTrack
       sections [{
         type=MovieScene2DTransformSection
         channels [
          Translation.X {type=MovieSceneScriptingFloatChannel}
          Translation.Y {type=MovieSceneScriptingFloatChannel}
          Angle {type=MovieSceneScriptingFloatChannel}
          Scale.X {type=MovieSceneScriptingFloatChannel, keys [[-0.2563,1],[14.6277,0.7824]]}
          Scale.Y {type=MovieSceneScriptingFloatChannel, keys [[-0.2563,1],[14.6277,0.7824]]}
          Shear.X {type=MovieSceneScriptingFloatChannel}
          Shear.Y {type=MovieSceneScriptingFloatChannel}
         ]
        }]
      }
      {
       type=MovieSceneFloatTrack
       display_name=Render Opacity
       sections [{type=MovieSceneFloatSection, channels [None {type=MovieSceneScriptingFloatChannel, keys [[23.6743,1],[24.3143,0]]}]}]
      }
     ]
    }
    Overlay_NewType {
     id=#6
     possessed_class=Overlay
     tracks [
      {
       type=MovieScene2DTransformTrack
       sections [{
         type=MovieScene2DTransformSection
         channels [
          Translation.X {type=MovieSceneScriptingFloatChannel}
          Translation.Y {type=MovieSceneScriptingFloatChannel}
          Angle {type=MovieSceneScriptingFloatChannel}
          Scale.X {type=MovieSceneScriptingFloatChannel, keys [[27.419,0],[31.6997,1],[32.419,1]]}
          Scale.Y {type=MovieSceneScriptingFloatChannel, keys [[27.419,0],[31.6997,1],[32.419,1]]}
          Shear.X {type=MovieSceneScriptingFloatChannel}
          Shear.Y {type=MovieSceneScriptingFloatChannel}
         ]
        }]
      }
      {
       type=MovieSceneFloatTrack
       display_name=Render Opacity
       sections [{type=MovieSceneFloatSection, channels [None {type=MovieSceneScriptingFloatChannel, keys [[27.419,0],[31.177,1]]}]}]
      }
     ]
    }
    Text_FirstCatch {
     id=#7
     possessed_class=CommonTextBlock
     tracks [
      {
       type=MovieScene2DTransformTrack
       sections [{
         type=MovieScene2DTransformSection
         channels [
          Translation.X {type=MovieSceneScriptingFloatChannel}
          Translation.Y {type=MovieSceneScriptingFloatChannel}
          Angle {type=MovieSceneScriptingFloatChannel}
          Scale.X {type=MovieSceneScriptingFloatChannel, keys [[28.984,1],[31,1.5],[32.6413,1]]}
          Scale.Y {type=MovieSceneScriptingFloatChannel, keys [[28.984,1],[31,1.5],[32.6413,1]]}
          Shear.X {type=MovieSceneScriptingFloatChannel}
          Shear.Y {type=MovieSceneScriptingFloatChannel}
         ]
        }]
      }
      {
       type=MovieSceneFloatTrack
       display_name=Render Opacity
       sections [{type=MovieSceneFloatSection, channels [None {type=MovieSceneScriptingFloatChannel, keys [[28.984,0],[29.489,1]]}]}]
      }
      {
       type=MovieSceneWidgetMaterialTrack
       display_name=Font.FontMaterial
       sections [{type=MovieSceneParameterSection, channels [TimeOffset {type=MovieSceneScriptingFloatChannel, keys [[28.984,0.5071],[55.997,-1.3581]]}]}]
      }
     ]
    }
    Image_star {
     id=#8
     possessed_class=Image
     tracks [{
       type=MovieScene2DTransformTrack
       sections [{
         type=MovieScene2DTransformSection
         channels [
          Translation.X {type=MovieSceneScriptingFloatChannel}
          Translation.Y {type=MovieSceneScriptingFloatChannel}
          Angle {type=MovieSceneScriptingFloatChannel}
          Scale.X {type=MovieSceneScriptingFloatChannel, keys [[27.419,1],[39.419,0.5]]}
          Scale.Y {type=MovieSceneScriptingFloatChannel, keys [[27.419,1],[39.419,0.5]]}
          Shear.X {type=MovieSceneScriptingFloatChannel}
          Shear.Y {type=MovieSceneScriptingFloatChannel}
         ]
        }]
      }]
    }
    Image_textglow {
     id=#9
     possessed_class=Image
     tracks [{
       type=MovieSceneFloatTrack
       display_name=Render Opacity
       sections [{type=MovieSceneFloatSection, channels [None {type=MovieSceneScriptingFloatChannel, keys [[27.419,0],[31.177,1]]}]}]
      }]
    }
    Overlay_TopContainer {
     id=#10
     possessed_class=Overlay
     tracks [{
       type=MovieSceneFloatTrack
       display_name=Render Opacity
       sections [{type=MovieSceneFloatSection, channels [None {type=MovieSceneScriptingFloatChannel, keys [[-0.2563,0],[1.769,1]]}]}]
      }]
    }
   ]
   tracks [
    {
     type=MovieSceneEventTrack
     sections [{
       type=MovieSceneEventTriggerSection
       channels [None {type=MovieSceneScriptingEventChannel, const ["<Struct 'MovieSceneEvent' (0x4679f02eba0) {}>", 1]}]
      }]
    }
    {
     type=MovieSceneEventTrack
     sections [{
       type=MovieSceneEventTriggerSection
       channels [None {type=MovieSceneScriptingEventChannel, const ["<Struct 'MovieSceneEvent' (0x4679f02fc80) {}>", 1]}]
      }]
    }
   ]
  }
 }
 LeaderTransition {
  binding_count=4
  length_seconds=0.9121
  object=&self:LeaderTransition
  sequence {bindings [
    Overlay_NewType {
     id=#11
     possessed_class=Overlay
     tracks [{
       type=MovieScene2DTransformTrack
       sections [{
         type=MovieScene2DTransformSection
         channels [
          Translation.X {type=MovieSceneScriptingFloatChannel}
          Translation.Y {type=MovieSceneScriptingFloatChannel}
          Angle {type=MovieSceneScriptingFloatChannel}
          Scale.X {type=MovieSceneScriptingFloatChannel, keys [[0,1],[6.45,0]]}
          Scale.Y {type=MovieSceneScriptingFloatChannel, keys [[0,1],[6.45,0]]}
          Shear.X {type=MovieSceneScriptingFloatChannel}
          Shear.Y {type=MovieSceneScriptingFloatChannel}
         ]
        }]
      }]
    }
    CommonBorder_ItemName {
     id=#12
     possessed_class=CommonBorder
     tracks [{
       type=MovieSceneFloatVectorTrack
       display_name=Desired Size Scale
       sections [{
         type=MovieSceneFloatVectorSection
         channels [
          Vector.X {type=MovieSceneScriptingFloatChannel}
          Vector.Y {type=MovieSceneScriptingFloatChannel, keys [[0,1],[11.2,0]]}
         ]
        }]
      }]
    }
    Border_Leader {
     id=#13
     possessed_class=CommonBorder
     tracks [{
       type=MovieSceneFloatVectorTrack
       display_name=Desired Size Scale
       sections [{
         type=MovieSceneFloatVectorSection
         channels [
          Vector.X {type=MovieSceneScriptingFloatChannel}
          Vector.Y {type=MovieSceneScriptingFloatChannel, keys [[7.042,0],[18.242,1]]}
         ]
        }]
      }]
    }
    Switcher_SizeInfo {
     id=#14
     possessed_class=CommonWidgetSwitcherLegacy
     tracks [{
       type=MovieSceneIntegerTrack
       display_name=Active Widget Index
       sections [{type=MovieSceneIntegerSection, channels [None {type=MovieSceneScriptingIntegerChannel, keys [[0,0],[3,1]]}]}]
      }]
    }
   ]}
 }
 StarScale {
  binding_count=3
  length_seconds=1.5106
  object=&self:StarScale
  sequence {bindings [
    Image_StarRating {
     id=#15
     possessed_class=FortLazyImage
     tracks [{
       type=MovieSceneWidgetMaterialTrack
       sections [{
         type=MovieSceneParameterSection
         channels [
          Star1Scale {type=MovieSceneScriptingFloatChannel, keys [[0,0],[5,2.3361],[7,0.6986],[9,1]]}
          Star2Scale {type=MovieSceneScriptingFloatChannel, keys [[5,0],[10,2.3361],[12,0.6986],[14,1]]}
          Star3Scale {type=MovieSceneScriptingFloatChannel, keys [[12.0753,0],[17.0753,2.3361],[19.0753,0.6986],[21.0753,1]]}
          Icon_strenght {type=MovieSceneScriptingFloatChannel, keys [[21.0753,0],[30.211,0.8]]}
         ]
        }]
      }]
    }
    Image_3starglow {
     id=#16
     possessed_class=Image
     tracks [
      {
       type=MovieScene2DTransformTrack
       sections [{
         type=MovieScene2DTransformSection
         channels [
          Translation.X {type=MovieSceneScriptingFloatChannel}
          Translation.Y {type=MovieSceneScriptingFloatChannel}
          Angle {type=MovieSceneScriptingFloatChannel}
          Scale.X {type=MovieSceneScriptingFloatChannel, keys [[0.061,0.8],[30.211,1]]}
          Scale.Y {type=MovieSceneScriptingFloatChannel, keys [[0.061,0.8],[30.211,1]]}
          Shear.X {type=MovieSceneScriptingFloatChannel}
          Shear.Y {type=MovieSceneScriptingFloatChannel}
         ]
        }]
      }
      {
       type=MovieSceneFloatTrack
       display_name=Render Opacity
       sections [{type=MovieSceneFloatSection, channels [None {type=MovieSceneScriptingFloatChannel, keys [[0,0],[30.211,1]]}]}]
      }
     ]
    }
    Overlay_Stars {
     id=#17
     possessed_class=Overlay
     tracks [{
       type=MovieSceneFloatTrack
       display_name=Render Opacity
       sections [{type=MovieSceneFloatSection, channels [None {type=MovieSceneScriptingFloatChannel, keys [[0,0],[1.132,1]]}]}]
      }]
    }
   ]}
 }
]
graphs [
 EventGraph {nodes [
   ResetAnims() {is=title K2Node_CustomEvent}
   > StopAllAnimations(self:User Widget Object Reference)
   n1 = Get StarScale() -> (StarScale:Widget Animation Object Reference)
   > PlayAnimation(InAnimation=@n1, StartAtTime=GetEndTime(self=@n1), PlayMode=Reverse)
   n2 = Get LeaderTransition() -> (LeaderTransition:Widget Animation Object Reference)
   > PlayAnimation(InAnimation=@n2, StartAtTime=GetEndTime(self=@n2), PlayMode=Reverse)
   n3 = Get NewFish() -> (NewFish:Widget Animation Object Reference)
   > PlayAnimation(InAnimation=@n3, StartAtTime=GetEndTime(self=@n3), PlayMode=Reverse)
   n4 = Event On Star Rating Set!() -> (OutputDelegate, then, StarRating:int=0) {is=title K2Node_Event}
   > Set Star Rating(Star Rating:int=@n4.StarRating) -> (Output_Get:int=0)
   > n5 = Branch(Condition=Is Design Time!(self:Athena Collection Item Toast Object Reference) -> (ReturnValue:bool))
   n6 = GetDynamicMaterial(execute=@n5.else, self=Get Image_StarRating() -> (Image_StarRating:Fort Lazy Image Object Reference))
   > SetScalarParameterValue(
    self=@n6
    ParameterName=Is2Stars
    Value=Select(
     Option 0:double=0
     Option 1:double=1
     Index="integer >= integer"!(A:int=@n4.StarRating, B:int=2) -> (ReturnValue:bool) {is=title K2Node_CallFunction}
    ) -> (ReturnValue:double)
   )
   n7 = "Equal (Integer)"(A=@n4.StarRating, B=3) {is=title K2Node_CallFunction}
   > SetScalarParameterValue(self=@n6, ParameterName=Is3Stars, Value=Select(Option 0=0, Option 1=1, Index=@n7))
   > SetVisibility(
    self=Get Image_3starglow() -> (Image_3starglow:Image Object Reference)
    InVisibility=Select(Option 0:ESlateVisibility Enum=Collapsed, Option 1:ESlateVisibility Enum=HitTestInvisible, Index=@n7) -> (ReturnValue:ESlateVisibility Enum)
   )
   n8 = Event On Toast Updated!() -> (OutputDelegate, then, CollectedState:EFortCollectedState Enum=Unknown, NumCollected:int=0) {is=title K2Node_Event}
   > n9 = Branch(Condition=Is Design Time())
   Set IsToastInProgress(execute=@n9.else, IsToastInProgress:bool=true) -> (Output_Get:bool)
   > SetActiveWidgetIndex(
    self=[
     Get Switcher_ClueText() -> (Switcher_ClueText:Common Widget Switcher Legacy Object Reference)
     Get Switcher_MobileClueText() -> (Switcher_MobileClueText:Common Widget Switcher Legacy Object Reference)
    ]
    Index=Select(Option 0:int=0, Option 1:int=1, Index=Get IsClueQueued() -> (IsClueQueued:bool)) -> (ReturnValue:int)
   )
   > Set IsClueQueued(IsClueQueued:bool=false) -> (Output_Get:bool)
   n10 = "integer > integer"(A=@n8.NumCollected, B=1) {is=title K2Node_CallFunction}
   > SetVisibility(
    self=Get Text_FirstCatch() -> (Text_FirstCatch:Common Text Object Reference)
    InVisibility=Select(Option 0:ESlateVisibility Enum=SelfHitTestInvisible, Option 1:ESlateVisibility Enum=Collapsed, Index=@n10) -> (ReturnValue:ESlateVisibility Enum)
   )
   n11 = Select!(
    Unknown:int=-1
    New:int=0
    Known:int=-1
    NewlyCollected:int=Select!(Option 0:int=0, Option 1:int=-1, Index:bool=@n10) -> (ReturnValue)
    Collected:int=-1
    NewBest:int=1
    NewRecord:int=2
    NewLocation:int=0
    NewlyCompleted:int=0
    Complete:int=0
    Index:EFortCollectedState Enum=@n8.CollectedState
   ) -> (ReturnValue:int)
   > SetActiveWidgetIndex(
    self:Common Visibility Switcher Object Reference=Get VisibilitySwitcher_ToastType() -> (VisibilitySwitcher_ToastType:Common Visibility Switcher Object Reference)
    Index=@n11
   )
   > SetVisibility(
    self=Get Overlay_NewType() -> (Overlay_NewType:Overlay Object Reference)
    InVisibility=Select(Option 0:ESlateVisibility Enum=SelfHitTestInvisible, Option 1:ESlateVisibility Enum=Collapsed, Index="Equal (Integer)"(A=@n11, B=-1)) -> (ReturnValue:ESlateVisibility Enum)
   )
   > ResetAnims() {is=title K2Node_CallFunction}
   n12 = Literal enum EFortCollectedState!(Enum:EFortCollectedState Enum=Known) -> (ReturnValue:EFortCollectedState Enum) {is=title K2Node_EnumLiteral}
   > SetVisibility(
    self=Get Image_StarRating()
    InVisibility=Select(Option 0=Collapsed, Option 1=SelfHitTestInvisible, Index="Byte > Byte"!(A:Byte=@n8.CollectedState, B:Byte=@n12) -> (ReturnValue:bool))
   )
   > Set CollectedState(CollectedState:EFortCollectedState Enum=@n8.CollectedState) -> (Output_Get:EFortCollectedState Enum=Unknown)
   n13 = Event On Normalized Length Set!() -> (OutputDelegate, then, NormalizedLength:float, NormalizedMinLength:float, NormalizedMaxLength:float) {is=title K2Node_Event}
   n14 = Get Image_RelativeSizePointer() -> (Image_RelativeSizePointer:Image Object Reference)
   n15 = MapRangeClamped(Value=@n13.NormalizedLength, InRangeA=0, InRangeB=1, OutRangeA=@n13.NormalizedMinLength, OutRangeB=@n13.NormalizedMaxLength)
   n16 = Get Image_RelativeSizeMeter() -> (Image_RelativeSizeMeter:Image Object Reference)
   n17 = BreakVector2D(InVec=Break Slate Brush!(SlateBrush:"Slate Brush Structure (by ref)"=Get Brush!(self:Image Object Reference=@n16) -> (Brush:Slate Brush Structure)) -> (
     TintColor:Slate Color Structure
     DrawAs:ESlateBrushDrawType Enum
     Tiling:ESlateBrushTileType Enum
     Mirroring:ESlateBrushMirrorType Enum
     ImageSize:Vector 2D Structure
     Margin:Margin Structure
     ResourceObject:Object Reference
     OutlineSettings:Slate Brush Outline Settings Structure
    ) {is=title K2Node_BreakStruct}.ImageSize)
   n18 = float * float(A=@n13.NormalizedMaxLength, B=@n17)
   n19 = float * float(A=@n13.NormalizedMinLength, B=@n17)
   > SetRenderTranslation!(
    execute
    self=@n14
    Translation_X:double=float + float(A=Lerp(A=-11, B=-15, Alpha=@n15), B=Lerp(A=@n19, B=@n18, Alpha=@n13.NormalizedLength)) {is=title K2Node_CommutativeAssociativeBinaryOperator}
    Translation_Y:double=BreakVector2D(InVec=Get!(self:Widget Object Reference=@n14) -> (
      RenderTransform_Translation:Vector 2D Structure
      RenderTransform_Scale:Vector 2D Structure
      RenderTransform_Shear:Vector 2D Structure
      RenderTransform_Angle:float=~
     )).Y
   ) -> (then)
   > n20 = GetDynamicMaterial(self=Get Image_RelativeSizeMeter())
   > SetScalarParameterValue(self=@n20, ParameterName=MinSize, Value=@n13.NormalizedMinLength)
   > SetScalarParameterValue(self=@n20, ParameterName=MaxSize, Value=@n13.NormalizedMaxLength)
   Event On Toast Reset() {is=title K2Node_Event}
   > SetVisibility(
    self=[Get Text_FirstCatch(), Get Text_NewClue() -> (Text_NewClue:Common Text Object Reference), Get Overlay_NewType(), Get Image_StarRating()]
    InVisibility=Collapsed
   )
   > Set IsToastInProgress(IsToastInProgress=false)
   n21 = Event On Rank Set!() -> (OutputDelegate, then, Rank:int=0, NumEntries:int=0) {is=title K2Node_Event}
   > SetActiveWidgetIndex(
    self:Widget Switcher Object Reference=Get Switcher_LeaderType() -> (Switcher_LeaderType:Common Widget Switcher Legacy Object Reference)
    Index=Select(Option 0:int=0, Option 1:int=1, Index="Equal (Integer)"(A=@n21.Rank, B=1)) -> (ReturnValue:int)
   )
   Event On Unlocked New Clue() {is=title K2Node_Event}
   > n22 = Branch(Condition=Get IsToastInProgress() -> (IsToastInProgress:bool))
   > SetActiveWidgetIndex(self=[Get Switcher_ClueText(), Get Switcher_MobileClueText()], Index=1)
   Set IsClueQueued(execute=@n22.else, IsClueQueued=true)
   Event On Display Toast() {is=title K2Node_Event}
   n23 = Get NewFish()
   > PlayAnimation(InAnimation=@n23)
   > RetriggerableDelay(Duration=float + float(A=GetEndTime(self=@n23), B=2))
   n24 = Get LeaderTransition()
   > PlayAnimation(InAnimation=@n24)
   > RetriggerableDelay(Duration=float + float(A=GetEndTime(self=@n24), B=2))
  ]}
 Sequencer Events {nodes [
   ShowStars() {is=title K2Node_CustomEvent}
   > PlayAnimation(InAnimation=Get StarScale())
   > PlaySound2D(Sound=Select!(
     Option 0:Sound Base Object Reference=/Game/Athena/Sounds/Items/Flopper/Flopping_Rabbit_Toast_1_Star_Cue
     Option 1:Sound Base Object Reference=&Flopping_Rabbit_Toast_1_Star_Cue
     Option 2:Sound Base Object Reference=/Game/Athena/Sounds/Items/Flopper/Flopping_Rabbit_Toast_2_Star_Cue
     Option 3:Sound Base Object Reference=/Game/Athena/Sounds/Items/Flopper/Flopping_Rabbit_Toast_3_Star_Cue
     Index:int=Get Star Rating() -> (Star Rating:int=0)
    ) -> (ReturnValue:Sound Base Object Reference))
   > Set Star Rating(Star Rating=0) -> (Output_Get=0)
   Play Catch Audio() {is=title K2Node_CustomEvent}
   n1 = Select!(
    Unknown:Sound Base Object Reference
    New:Sound Base Object Reference=/Game/Athena/Sounds/Items/Flopper/Flopping_Rabbit_Toast_First_Catch
    Known:Sound Base Object Reference
    NewlyCollected:Sound Base Object Reference=/Game/Athena/Sounds/Items/Flopper/Flopping_Rabbit_Toast_Newly_Collected_Cue
    Collected:Sound Base Object Reference
    NewBest:Sound Base Object Reference=&Flopping_Rabbit_Toast_Newly_Collected_Cue
    NewRecord:Sound Base Object Reference=&Flopping_Rabbit_Toast_Newly_Collected_Cue
    NewLocation:Sound Base Object Reference
    NewlyCompleted:Sound Base Object Reference
    Complete:Sound Base Object Reference
    Index:EFortCollectedState Enum=Get CollectedState() -> (CollectedState:EFortCollectedState Enum=Unknown)
   ) -> (ReturnValue)
   > Is Valid(InputObject=@n1)
   > PushSoundMixModifier!(execute, InSoundMixModifier:Sound Mix Object Reference=/Game/Sounds/FortSoundClassesAndModes/Suppress_Punchcard_UI) -> (then)
   > PlaySound2D(Sound=@n1)
  ]}
]
variables [
 CollectedState {type=EFortCollectedState, default=Unknown}
 IsClueQueued {type=boolean}
 IsToastInProgress {type=boolean}
 Star Rating {type=integer}
]
widgets {
 info {inheritedWidgetCount=8, parentClass=/Script/FortniteUI.AthenaCollectionItemToastFishing, rootWidgetClass=&SizeBox, widgetCount=69}
 widgets [
  SizeBox_1 {widgetClassPath=&SizeBox, widget=SizeBox_1, properties {bOverride_HeightOverride=true, heightOverride=128}}
  Overlay_3 {widgetClassPath=&Overlay, parent=SizeBox_1, widget=Overlay_3, slot {class=SizeBoxSlot}}
  Overlay_TopContainer {
   widgetClassPath=&Overlay
   parent=Overlay_3
   widget=Overlay_TopContainer
   slot {class=OverlaySlot, properties {horizontalAlignment=HAlign_Center, verticalAlignment=VAlign_Center}}
  }
  Overlay_Background {
   widgetClassPath=&Overlay
   parent=Overlay_TopContainer
   widget=Overlay_Background
   properties {visibility=Visible}
   slot {class=OverlaySlot, properties {horizontalAlignment=HAlign_Fill, verticalAlignment=VAlign_Fill, padding=<ltrb 0 20 0 0>}}
  }
  SizeBox_0 {
   widgetClassPath=&SizeBox
   parent=Overlay_Background
   widget=SizeBox_0
   properties {bOverride_HeightOverride=true, heightOverride=73, renderTransform {angle=0, scale=<1 1>, shear=<0 0>, translation=<0 -9>}}
   slot {class=OverlaySlot, properties {horizontalAlignment=HAlign_Fill, verticalAlignment=VAlign_Center}}
  }
  Border_Background {
   widgetClassPath=&CommonBorder
   parent=SizeBox_0
   widget=Border_Background
   properties {
    background {
     bIsDynamicallyLoaded=false
     drawAs=Box
     imageType=NoImage
     mirroring=NoMirror
     resourceName=None
     resourceObject=/Game/Athena/HUD/Collections/Materials/M_UI_BackgroundWShadow_Inst
     tiling=NoTile
     imageSize=<128 128>
     margin=<ltrb 0.2 0 0.25 0>
     outlineSettings {
      bUseBrushTransparency=false
      roundingType=HalfHeightRadius
      width=0
      color {colorUseRule=UseColor_Specified, specifiedColor=<rgba 0 0 0 0>}
      cornerRadii=<0 0 0 0>
     }
     tintColor {colorUseRule=UseColor_Specified, specifiedColor=<rgba 1 1 1 1>}
     uVRegion {bIsValid=false, max=<0 0>, min=<0 0>}
    }
    padding=<ltrb 0 0 0 0>
   }
   slot {class=SizeBoxSlot, properties {padding=<ltrb 0 0 -40 -10>}}
  }
  Button_MobileClickCatcher {
   widgetClassPath=&Button
   bInherited=true
   parent=Border_Background
   widget=Button_MobileClickCatcher
   properties {isFocusable=false, backgroundColor=<rgba 1 1 1 0>, colorAndOpacity=<rgba 1 1 1 0>}
   slot {class=BorderSlot, properties {padding=<ltrb 0 0 0 0>}}
   bIsVariable=~
  }
  HorizontalBoxHotkeyPrompt {
   widgetClassPath=&HorizontalBox
   parent=Overlay_Background
   widget=HorizontalBoxHotkeyPrompt
   properties {visibility=Visible, renderTransform {angle=0, scale=<1 1>, shear=<0 0>, translation=<0 15>}}
   slot {class=OverlaySlot, properties {horizontalAlignment=HAlign_Right, verticalAlignment=VAlign_Bottom}}
  }
  SizeBox_Hotkey {
   widgetClassPath=&SizeBox
   parent=HorizontalBoxHotkeyPrompt
   widget=SizeBox_Hotkey
   properties {bOverride_HeightOverride=true, bOverride_WidthOverride=true, heightOverride=48, widthOverride=48}
   slot {class=HorizontalBoxSlot, properties {horizontalAlignment=HAlign_Right, verticalAlignment=VAlign_Top}}
  }
  ScaleBox_0 {widgetClassPath=&ScaleBox, parent=SizeBox_Hotkey, widget=ScaleBox_0, slot {class=SizeBoxSlot}}
  Overlay_1 {widgetClassPath=&Overlay, parent=ScaleBox_0, widget=Overlay_1, slot {class=ScaleBoxSlot}}
  Gamepad {
   widgetClassPath=&CommonVisibilityWidgetLegacy
   parent=Overlay_1
   widget=Gamepad
   properties {bShowForMouseAndKeyboard=false, bShowForTouch=false, padding=<ltrb 0 0 0 0>}
   slot {class=OverlaySlot}
  }
  KeybindWidget {
   widgetClassPath=/FortUILibrary/Input/KeybindWidget
   parent=Gamepad
   widget=KeybindWidget
   properties {boundAction=GamepadToggleFullScreenMap, forcedHoldKeybindStatus=ForcedHold}
   slot {class=BorderSlot, properties {padding=<ltrb 0 0 0 0>}}
  }
  MouseKeyboard {
   widgetClassPath=&CommonVisibilityWidgetLegacy
   parent=Overlay_1
   widget=MouseKeyboard
   properties {
    bShowForAndroid=false
    bShowForGamepad=false
    bShowForIOS=false
    bShowForPS4=false
    bShowForTouch=false
    bShowForXBox=false
    padding=<ltrb 0 0 0 0>
   }
   slot {class=OverlaySlot, properties {horizontalAlignment=HAlign_Fill, verticalAlignment=VAlign_Center}}
  }
  KeybindWidget_1 {
   widgetClassPath=&KeybindWidget
   parent=MouseKeyboard
   widget=KeybindWidget_1
   properties {boundAction=ToggleFullScreenMap, forcedHoldKeybindStatus=ForcedHold}
   slot {class=BorderSlot, properties {padding=<ltrb 0 0 0 0>}}
  }
  Visibility_MobileCollectionText {
   widgetClassPath=&CommonVisibilityWidgetLegacy
   parent=HorizontalBoxHotkeyPrompt
   widget=Visibility_MobileCollectionText
   properties {
    bShowForErebus=false
    bShowForGamepad=false
    bShowForMac=false
    bShowForMouseAndKeyboard=false
    bShowForPC=false
    bShowForPS4=false
    bShowForSwitch2=false
    bShowForXBox=false
   }
   slot {class=HorizontalBoxSlot, properties {horizontalAlignment=HAlign_Right, verticalAlignment=VAlign_Center}}
  }
  Switcher_MobileClueText {
   widgetClassPath=&CommonWidgetSwitcherLegacy
   bIsVariable=true
   parent=Visibility_MobileCollectionText
   widget=Switcher_MobileClueText
   properties {transitionDuration=0}
   slot {class=BorderSlot}
  }
  Text_MobileCollectionPrompt {
   widgetClassPath=&CommonTextBlock
   bIsVariable=true
   parent=Switcher_MobileClueText
   widget=Text_MobileCollectionPrompt
   properties {
    text=TAP TO OPEN COLLECTION
    textTransformPolicy=ToUpper
    font {
     bForceMonospaced=false
     bMaterialIsStencil=false
     fontMaterial=null
     fontObject=&Burbank
     letterSpacing=0
     monospacedWidth=1
     size=12
     skewAmount=0
     typefaceFontName=Burbank-PlayerPower
     outlineSettings {
      bApplyOutlineToDropShadows=false
      bMiteredCorners=false
      bSeparateFillAlpha=false
      outlineMaterial=null
      outlineSize=0
      outlineColor=<rgba 0 0 0 1>
     }
    }
    renderTransform {angle=0, scale=<1 1>, shear=<-12 0>, translation=<0 0>}
    shadowOffset=<0 0>
   }
   slot {class=WidgetSwitcherSlot}
  }
  Overlay_MobileClue {
   widgetClassPath=&Overlay
   parent=Switcher_MobileClueText
   widget=Overlay_MobileClue
   slot {class=WidgetSwitcherSlot, properties {verticalAlignment=VAlign_Center, padding=<ltrb 8 0 16 0>}}
  }
  Text_NewClue {
   widgetClassPath=&CommonTextBlock
   bIsVariable=true
   parent=Overlay_MobileClue
   widget=Text_NewClue
   properties {
    minDesiredWidth=0.5
    text="Tap for New Clue!"
    textTransformPolicy=ToUpper
    colorAndOpacity {colorUseRule=UseColor_Specified, specifiedColor=<rgba 1 0.6445 0 1>}
    font {
     bForceMonospaced=false
     bMaterialIsStencil=false
     fontMaterial=null
     fontObject=&Burbank
     letterSpacing=0
     monospacedWidth=1
     size=12
     skewAmount=0
     typefaceFontName=Burbank-PlayerPower
     outlineSettings {
      bApplyOutlineToDropShadows=false
      bMiteredCorners=false
      bSeparateFillAlpha=false
      outlineMaterial=&M-FontShadow
      outlineSize=3
      outlineColor=<rgba 0 0 0 0.8>
     }
    }
    renderTransform {angle=0, scale=<1 1>, shear=<-8 0>, translation=<0 0>}
    shadowOffset=<2 2>
   }
   slot {class=OverlaySlot, properties {horizontalAlignment=HAlign_Center, verticalAlignment=VAlign_Center, padding=<ltrb 0 0 -2 -2>}}
  }
  Image_glow {
   widgetClassPath=&Image
   parent=Overlay_MobileClue
   widget=Image_glow
   properties {
    brush {
     bIsDynamicallyLoaded=false
     drawAs=Image
     imageType=NoImage
     mirroring=NoMirror
     resourceName=None
     resourceObject=/Game/Athena/HUD/LTM/Wax/Materials/M_UI_TextGlow_Pulse_Add
     tiling=NoTile
     imageSize=<32 32>
     margin=<ltrb 0 0 0 0>
     outlineSettings {
      bUseBrushTransparency=false
      roundingType=HalfHeightRadius
      width=0
      color {colorUseRule=UseColor_Specified, specifiedColor=<rgba 0 0 0 0>}
      cornerRadii=<0 0 0 0>
     }
     tintColor {colorUseRule=UseColor_Specified, specifiedColor=<rgba 1 1 1 1>}
     uVRegion {bIsValid=false, max=<0 0>, min=<0 0>}
    }
    colorAndOpacity=<rgba 0.2159 0.0467 0 1>
   }
   slot {class=OverlaySlot, properties {horizontalAlignment=HAlign_Fill, verticalAlignment=VAlign_Fill, padding=<ltrb -30 0 -30 0>}}
  }
  Visibility_CollectionText {
   widgetClassPath=&CommonVisibilityWidgetLegacy
   parent=HorizontalBoxHotkeyPrompt
   widget=Visibility_CollectionText
   properties {bShowForAndroid=false, bShowForIOS=false, bShowForTouch=false}
   slot {class=HorizontalBoxSlot, properties {horizontalAlignment=HAlign_Right, verticalAlignment=VAlign_Center, padding=<ltrb 8 0 0 0>}}
  }
  Switcher_ClueText {
   widgetClassPath=&CommonWidgetSwitcherLegacy
   bIsVariable=true
   parent=Visibility_CollectionText
   widget=Switcher_ClueText
   properties {transitionDuration=0}
   slot {class=BorderSlot}
  }
  Text_CollectionPrompt {
   widgetClassPath=&CommonTextBlock
   bIsVariable=true
   parent=Switcher_ClueText
   widget=Text_CollectionPrompt
   properties {
    text=Collection
    textTransformPolicy=ToUpper
    font {
     bForceMonospaced=false
     bMaterialIsStencil=false
     fontMaterial=null
     fontObject=&Burbank
     letterSpacing=0
     monospacedWidth=1
     size=12
     skewAmount=0
     typefaceFontName=Burbank-PlayerPower
     outlineSettings {
      bApplyOutlineToDropShadows=false
      bMiteredCorners=false
      bSeparateFillAlpha=false
      outlineMaterial=null
      outlineSize=0
      outlineColor=<rgba 0 0 0 1>
     }
    }
    renderTransform {angle=0, scale=<1 1>, shear=<-8 0>, translation=<0 0>}
    shadowOffset=<0 0>
   }
   slot {class=WidgetSwitcherSlot}
  }
  Overlay_Clue {
   widgetClassPath=&Overlay
   parent=Switcher_ClueText
   widget=Overlay_Clue
   slot {class=WidgetSwitcherSlot, properties {verticalAlignment=VAlign_Center, padding=<ltrb 8 0 16 0>}}
  }
  Text_NewClue_1 {
   widgetClassPath=&CommonTextBlock
   bIsVariable=true
   parent=Overlay_Clue
   widget=Text_NewClue_1
   properties {
    minDesiredWidth=0.5
    text="New Clue!"
    textTransformPolicy=ToUpper
    colorAndOpacity {colorUseRule=UseColor_Specified, specifiedColor=<rgba 1 0.6445 0 1>}
    font {
     bForceMonospaced=false
     bMaterialIsStencil=false
     fontMaterial=null
     fontObject=&Burbank
     letterSpacing=0
     monospacedWidth=1
     size=12
     skewAmount=0
     typefaceFontName=Burbank-PlayerPower
     outlineSettings {
      bApplyOutlineToDropShadows=false
      bMiteredCorners=false
      bSeparateFillAlpha=false
      outlineMaterial=&M-FontShadow
      outlineSize=3
      outlineColor=<rgba 0 0 0 0.8>
     }
    }
    renderTransform {angle=0, scale=<1 1>, shear=<-8 0>, translation=<0 0>}
    shadowOffset=<2 2>
   }
   slot {class=OverlaySlot, properties {horizontalAlignment=HAlign_Center, verticalAlignment=VAlign_Center, padding=<ltrb 0 0 -2 -2>}}
  }
  Image_glow_1 {
   widgetClassPath=&Image
   parent=Overlay_Clue
   widget=Image_glow_1
   properties {
    brush {
     bIsDynamicallyLoaded=false
     drawAs=Image
     imageType=NoImage
     mirroring=NoMirror
     resourceName=None
     resourceObject=&M_UI_TextGlow_Pulse_Add
     tiling=NoTile
     imageSize=<32 32>
     margin=<ltrb 0 0 0 0>
     outlineSettings {
      bUseBrushTransparency=false
      roundingType=HalfHeightRadius
      width=0
      color {colorUseRule=UseColor_Specified, specifiedColor=<rgba 0 0 0 0>}
      cornerRadii=<0 0 0 0>
     }
     tintColor {colorUseRule=UseColor_Specified, specifiedColor=<rgba 1 1 1 1>}
     uVRegion {bIsValid=false, max=<0 0>, min=<0 0>}
    }
    colorAndOpacity=<rgba 0.2159 0.0467 0 1>
   }
   slot {class=OverlaySlot, properties {horizontalAlignment=HAlign_Fill, verticalAlignment=VAlign_Fill, padding=<ltrb -30 0 -30 0>}}
  }
  HorizontalBox_ImageAndDetails {
   widgetClassPath=&HorizontalBox
   parent=Overlay_TopContainer
   widget=HorizontalBox_ImageAndDetails
   slot {class=OverlaySlot, properties {horizontalAlignment=HAlign_Fill, verticalAlignment=VAlign_Fill}}
  }
  Overlay_ItemImage {
   widgetClassPath=&Overlay
   parent=HorizontalBox_ImageAndDetails
   widget=Overlay_ItemImage
   slot {class=HorizontalBoxSlot, properties {horizontalAlignment=HAlign_Center, verticalAlignment=VAlign_Center}}
  }
  Image_textglow_1 {
   widgetClassPath=&Image
   parent=Overlay_ItemImage
   widget=Image_textglow_1
   properties {
    visibility=SelfHitTestInvisible
    brush {
     bIsDynamicallyLoaded=false
     drawAs=Image
     imageType=NoImage
     mirroring=NoMirror
     resourceName=None
     resourceObject=/Game/Athena/HUD/LTM/Wax/Materials/M_UI_TextGlow_Add
     tiling=NoTile
     imageSize=<105.9261 123.5972>
     margin=<ltrb 0 0 0 0>
     outlineSettings {
      bUseBrushTransparency=false
      roundingType=HalfHeightRadius
      width=0
      color {colorUseRule=UseColor_Specified, specifiedColor=<rgba 0 0 0 0>}
      cornerRadii=<0 0 0 0>
     }
     tintColor {colorUseRule=UseColor_Specified, specifiedColor=<rgba 0.0037 0.3613 1 1>}
     uVRegion {bIsValid=false, max=<0 0>, min=<0 0>}
    }
   }
   slot {class=OverlaySlot, properties {horizontalAlignment=HAlign_Center, verticalAlignment=VAlign_Center, padding=<ltrb -75 -75 -75 -75>}}
  }
  LazyImage_ItemImage {
   widgetClassPath=&FortLazyImage
   bInherited=true
   parent=Overlay_ItemImage
   widget=LazyImage_ItemImage
   properties {
    materialTextureParamName=Texture
    brush {
     bIsDynamicallyLoaded=false
     drawAs=Image
     imageType=NoImage
     mirroring=NoMirror
     resourceName=None
     resourceObject=/Game/Athena/HUD/Collections/Materials/M_UI_FishToast_Inst
     tiling=NoTile
     imageSize=<128 128>
     margin=<ltrb 0 0 0 0>
     outlineSettings {
      bUseBrushTransparency=false
      roundingType=HalfHeightRadius
      width=0
      color {colorUseRule=UseColor_Specified, specifiedColor=<rgba 0 0 0 0>}
      cornerRadii=<0 0 0 0>
     }
     tintColor {colorUseRule=UseColor_Specified, specifiedColor=<rgba 1 1 1 1>}
     uVRegion {bIsValid=false, max=<0 0>, min=<0 0>}
    }
    shadowColorAndOpacity=<rgba 0.0052 0.0052 0.0056 0.5>
   }
   slot {class=OverlaySlot, properties {horizontalAlignment=HAlign_Center, verticalAlignment=VAlign_Center}}
  }
  Image_Sparks {
   widgetClassPath=&Image
   parent=Overlay_ItemImage
   widget=Image_Sparks
   properties {
    renderOpacity=0
    visibility=SelfHitTestInvisible
    brush {
     bIsDynamicallyLoaded=false
     drawAs=Image
     imageType=NoImage
     mirroring=NoMirror
     resourceName=None
     resourceObject=/Game/Athena/UI/InGame/Interrogation/Materials/M_UI_Starburst_Inst3
     tiling=NoTile
     imageSize=<32 32>
     margin=<ltrb 0 0 0 0>
     outlineSettings {
      bUseBrushTransparency=false
      roundingType=HalfHeightRadius
      width=0
      color {colorUseRule=UseColor_Specified, specifiedColor=<rgba 0 0 0 0>}
      cornerRadii=<0 0 0 0>
     }
     tintColor {colorUseRule=UseColor_Specified, specifiedColor=<rgba 1 1 1 1>}
     uVRegion {bIsValid=false, max=<0 0>, min=<0 0>}
    }
   }
   slot {class=OverlaySlot, properties {horizontalAlignment=HAlign_Fill, verticalAlignment=VAlign_Fill, padding=<ltrb -100 -100 -100 -100>}}
  }
  Image_Swoosh {
   widgetClassPath=&Image
   parent=Overlay_ItemImage
   widget=Image_Swoosh
   properties {
    renderOpacity=0
    visibility=SelfHitTestInvisible
    brush {
     bIsDynamicallyLoaded=false
     drawAs=Image
     imageType=NoImage
     mirroring=NoMirror
     resourceName=None
     resourceObject=/Game/Athena/HUD/Collections/Materials/M_UI_PerspSwoosh_Inst
     tiling=NoTile
     imageSize=<200 200>
     margin=<ltrb 0 0 0 0>
     outlineSettings {
      bUseBrushTransparency=false
      roundingType=HalfHeightRadius
      width=0
      color {colorUseRule=UseColor_Specified, specifiedColor=<rgba 0 0 0 0>}
      cornerRadii=<0 0 0 0>
     }
     tintColor {colorUseRule=UseColor_Specified, specifiedColor=<rgba 1 1 1 1>}
     uVRegion {bIsValid=false, max=<0 0>, min=<0 0>}
    }
   }
   slot {class=OverlaySlot, properties {horizontalAlignment=HAlign_Center, verticalAlignment=VAlign_Center, padding=<ltrb -200 -200 -200 -200>}}
  }
  Image_spark1 {
   widgetClassPath=&Image
   parent=Overlay_ItemImage
   widget=Image_spark1
   properties {
    visibility=SelfHitTestInvisible
    brush {
     bIsDynamicallyLoaded=false
     drawAs=Image
     imageType=NoImage
     mirroring=NoMirror
     resourceName=None
     resourceObject=/Game/Athena/HUD/Materials/Victory/M_Flare_Cross_Animated
     tiling=NoTile
     imageSize=<32 32>
     margin=<ltrb 0 0 0 0>
     outlineSettings {
      bUseBrushTransparency=false
      roundingType=HalfHeightRadius
      width=0
      color {colorUseRule=UseColor_Specified, specifiedColor=<rgba 0 0 0 0>}
      cornerRadii=<0 0 0 0>
     }
     tintColor {colorUseRule=UseColor_Specified, specifiedColor=<rgba 1 1 1 1>}
     uVRegion {bIsValid=false, max=<0 0>, min=<0 0>}
    }
    renderTransform {angle=0, scale=<1 1>, shear=<0 0>, translation=<-28 -13>}
   }
   slot {class=OverlaySlot, properties {horizontalAlignment=HAlign_Center, verticalAlignment=VAlign_Center}}
  }
  Image_spark2 {
   widgetClassPath=&Image
   parent=Overlay_ItemImage
   widget=Image_spark2
   properties {
    visibility=SelfHitTestInvisible
    brush {
     bIsDynamicallyLoaded=false
     drawAs=Image
     imageType=NoImage
     mirroring=NoMirror
     resourceName=None
     resourceObject=&M_Flare_Cross_Animated
     tiling=NoTile
     imageSize=<16 16>
     margin=<ltrb 0 0 0 0>
     outlineSettings {
      bUseBrushTransparency=false
      roundingType=HalfHeightRadius
      width=0
      color {colorUseRule=UseColor_Specified, specifiedColor=<rgba 0 0 0 0>}
      cornerRadii=<0 0 0 0>
     }
     tintColor {colorUseRule=UseColor_Specified, specifiedColor=<rgba 1 1 1 1>}
     uVRegion {bIsValid=false, max=<0 0>, min=<0 0>}
    }
    renderTransform {angle=0, scale=<1 1>, shear=<0 0>, translation=<-5 -5>}
   }
   slot {class=OverlaySlot, properties {horizontalAlignment=HAlign_Center, verticalAlignment=VAlign_Center}}
  }
  Image_spark3 {
   widgetClassPath=&Image
   parent=Overlay_ItemImage
   widget=Image_spark3
   properties {
    visibility=SelfHitTestInvisible
    brush {
     bIsDynamicallyLoaded=false
     drawAs=Image
     imageType=NoImage
     mirroring=NoMirror
     resourceName=None
     resourceObject=&M_Flare_Cross_Animated
     tiling=NoTile
     imageSize=<32 32>
     margin=<ltrb 0 0 0 0>
     outlineSettings {
      bUseBrushTransparency=false
      roundingType=HalfHeightRadius
      width=0
      color {colorUseRule=UseColor_Specified, specifiedColor=<rgba 0 0 0 0>}
      cornerRadii=<0 0 0 0>
     }
     tintColor {colorUseRule=UseColor_Specified, specifiedColor=<rgba 1 1 1 1>}
     uVRegion {bIsValid=false, max=<0 0>, min=<0 0>}
    }
    renderTransform {angle=0, scale=<1 1>, shear=<0 0>, translation=<11 -14>}
   }
   slot {class=OverlaySlot, properties {horizontalAlignment=HAlign_Center, verticalAlignment=VAlign_Center}}
  }
  Overlay_Stars {
   widgetClassPath=&Overlay
   parent=Overlay_ItemImage
   widget=Overlay_Stars
   properties {renderOpacity=0, renderTransform {angle=8, scale=<1 1>, shear=<0 0>, translation=<-23 -36>}}
   slot {class=OverlaySlot, properties {horizontalAlignment=HAlign_Right, verticalAlignment=VAlign_Bottom, padding=<ltrb -55 -55 -55 -55>}}
  }
  Image_StarRating {
   widgetClassPath=&FortLazyImage
   bIsVariable=true
   parent=Overlay_Stars
   widget=Image_StarRating
   properties {
    bApplyDropShadow=true
    brush {
     bIsDynamicallyLoaded=false
     drawAs=Image
     imageType=NoImage
     mirroring=NoMirror
     resourceName=None
     resourceObject=/Game/Athena/Collections/Art/M_Ui_Collection_HighScore_IconBig_Toast
     tiling=NoTile
     imageSize=<116 116>
     margin=<ltrb 0 0 0 0>
     outlineSettings {
      bUseBrushTransparency=false
      roundingType=HalfHeightRadius
      width=0
      color {colorUseRule=UseColor_Specified, specifiedColor=<rgba 0 0 0 0>}
      cornerRadii=<0 0 0 0>
     }
     tintColor {colorUseRule=UseColor_Specified, specifiedColor=<rgba 1 1 1 1>}
     uVRegion {bIsValid=false, max=<0 0>, min=<0 0>}
    }
    shadowColorAndOpacity=<rgba 0 0 0 0.5>
   }
   slot {class=OverlaySlot}
  }
  Image_3starglow {
   widgetClassPath=&Image
   bIsVariable=true
   parent=Overlay_Stars
   widget=Image_3starglow
   properties {
    visibility=HitTestInvisible
    brush {
     bIsDynamicallyLoaded=false
     drawAs=Image
     imageType=NoImage
     mirroring=NoMirror
     resourceName=None
     resourceObject=&M_UI_TextGlow_Pulse_Add
     tiling=NoTile
     imageSize=<82.1221 71.9925>
     margin=<ltrb 0 0 0 0>
     outlineSettings {
      bUseBrushTransparency=false
      roundingType=HalfHeightRadius
      width=0
      color {colorUseRule=UseColor_Specified, specifiedColor=<rgba 0 0 0 0>}
      cornerRadii=<0 0 0 0>
     }
     tintColor {colorUseRule=UseColor_Specified, specifiedColor=<rgba 1 0.3468 0 1>}
     uVRegion {bIsValid=false, max=<0 0>, min=<0 0>}
    }
   }
   slot {class=OverlaySlot, properties {horizontalAlignment=HAlign_Center, verticalAlignment=VAlign_Center, padding=<ltrb -75 -75 -75 -75>}}
  }
  Border_ItemDetails {
   widgetClassPath=&CommonBorder
   parent=HorizontalBox_ImageAndDetails
   widget=Border_ItemDetails
   properties {clipping=ClipToBounds, padding=<ltrb 8 0 8 0>}
   slot {class=HorizontalBoxSlot, properties {verticalAlignment=VAlign_Center}}
  }
  VerticalBox_0 {
   widgetClassPath=&VerticalBox
   parent=Border_ItemDetails
   widget=VerticalBox_0
   properties {renderTransform {angle=0, scale=<1 1>, shear=<-8 0>, translation=<0 0>}}
   slot {class=BorderSlot, properties {padding=<ltrb 8 0 8 0>}}
  }
  CommonBorder_ItemName {
   widgetClassPath=&CommonBorder
   parent=VerticalBox_0
   widget=CommonBorder_ItemName
   properties {clipping=ClipToBounds, verticalAlignment=VAlign_Bottom, padding=<ltrb 0 0 0 0>}
   slot {class=VerticalBoxSlot}
  }
  HorizontalBox_58 {
   widgetClassPath=&HorizontalBox
   parent=CommonBorder_ItemName
   widget=HorizontalBox_58
   slot {class=BorderSlot, properties {verticalAlignment=VAlign_Bottom, padding=<ltrb 0 0 0 0>}}
  }
  CommonTextBlock_277 {
   widgetClassPath=&CommonTextBlock
   parent=HorizontalBox_58
   widget=CommonTextBlock_277
   properties {
    text=No.
    textTransformPolicy=ToUpper
    font {
     bForceMonospaced=false
     bMaterialIsStencil=false
     fontMaterial=null
     fontObject=&Burbank
     letterSpacing=0
     monospacedWidth=1
     size=12
     skewAmount=0
     typefaceFontName=Burbank-PlayerPower
     outlineSettings {
      bApplyOutlineToDropShadows=false
      bMiteredCorners=false
      bSeparateFillAlpha=false
      outlineMaterial=null
      outlineSize=0
      outlineColor=<rgba 0 0 0 1>
     }
    }
    shadowOffset=<0 0>
   }
   slot {class=HorizontalBoxSlot, properties {verticalAlignment=VAlign_Top, padding=<ltrb 0 2 1 0>}}
  }
  Text_ItemNumber {
   widgetClassPath=&CommonTextBlock
   bInherited=true
   parent=HorizontalBox_58
   widget=Text_ItemNumber
   properties {
    textTransformPolicy=ToUpper
    font {
     bForceMonospaced=false
     bMaterialIsStencil=false
     fontMaterial=null
     fontObject=&Burbank
     letterSpacing=0
     monospacedWidth=1
     size=16
     skewAmount=0
     typefaceFontName=Burbank-PlayerPower
     outlineSettings {
      bApplyOutlineToDropShadows=false
      bMiteredCorners=false
      bSeparateFillAlpha=false
      outlineMaterial=null
      outlineSize=0
      outlineColor=<rgba 0 0 0 1>
     }
    }
    shadowOffset=<0 0>
   }
   slot {class=HorizontalBoxSlot, properties {padding=<ltrb 0 0 8 0>}}
  }
  Text_ItemName {
   widgetClassPath=&CommonTextBlock
   bInherited=true
   parent=HorizontalBox_58
   widget=Text_ItemName
   properties {
    textTransformPolicy=ToUpper
    font {
     bForceMonospaced=false
     bMaterialIsStencil=false
     fontMaterial=null
     fontObject=&Burbank
     letterSpacing=0
     monospacedWidth=1
     size=16
     skewAmount=0
     typefaceFontName=Burbank-PlayerPower
     outlineSettings {
      bApplyOutlineToDropShadows=false
      bMiteredCorners=false
      bSeparateFillAlpha=false
      outlineMaterial=null
      outlineSize=0
      outlineColor=<rgba 0 0 0 1>
     }
    }
    shadowOffset=<0 0>
   }
   slot {class=HorizontalBoxSlot}
  }
  HorizontalBox_94 {widgetClassPath=&HorizontalBox, parent=VerticalBox_0, widget=HorizontalBox_94, slot {class=VerticalBoxSlot}}
  RichText_Length {
   widgetClassPath=&CommonRichTextBlock
   bInherited=true
   parent=HorizontalBox_94
   widget=RichText_Length
   properties {
    bOverrideDefaultStyle=true
    defaultTextStyleOverride {
     overflowPolicy=Clip
     transformPolicy=None
     colorAndOpacity {colorUseRule=UseColor_Specified, specifiedColor=<rgba 1 1 1 1>}
     font {
      bForceMonospaced=false
      bMaterialIsStencil=false
      fontMaterial=null
      fontObject=&Burbank
      letterSpacing=0
      monospacedWidth=1
      size=16
      skewAmount=0
      typefaceFontName=Burbank-PlayerPower
      outlineSettings {
       bApplyOutlineToDropShadows=false
       bMiteredCorners=false
       bSeparateFillAlpha=false
       outlineMaterial=null
       outlineSize=0
       outlineColor=<rgba 0 0 0 1>
      }
     }
     highlightColor {colorUseRule=UseColor_Specified, specifiedColor=<rgba 0 0 0 1>}
     highlightShape {
      bIsDynamicallyLoaded=false
      drawAs=Image
      imageType=NoImage
      mirroring=NoMirror
      resourceName=None
      resourceObject=null
      tiling=NoTile
      imageSize=<32 32>
      margin=<ltrb 0 0 0 0>
      outlineSettings {
       bUseBrushTransparency=false
       roundingType=HalfHeightRadius
       width=0
       color {colorUseRule=UseColor_Specified, specifiedColor=<rgba 0 0 0 0>}
       cornerRadii=<0 0 0 0>
      }
      tintColor {colorUseRule=UseColor_Specified, specifiedColor=<rgba 1 1 1 1>}
      uVRegion {bIsValid=false, max=<0 0>, min=<0 0>}
     }
     selectedBackgroundColor {colorUseRule=UseColor_Foreground, specifiedColor=<rgba 1 0 1 1>}
     shadowColorAndOpacity=<rgba 0 0 0 1>
     shadowOffset=<0 0>
     strikeBrush {
      bIsDynamicallyLoaded=false
      drawAs=Image
      imageType=NoImage
      mirroring=NoMirror
      resourceName=None
      resourceObject=null
      tiling=NoTile
      imageSize=<32 32>
      margin=<ltrb 0 0 0 0>
      outlineSettings {
       bUseBrushTransparency=false
       roundingType=HalfHeightRadius
       width=0
       color {colorUseRule=UseColor_Specified, specifiedColor=<rgba 0 0 0 0>}
       cornerRadii=<0 0 0 0>
      }
      tintColor {colorUseRule=UseColor_Specified, specifiedColor=<rgba 1 1 1 1>}
      uVRegion {bIsValid=false, max=<0 0>, min=<0 0>}
     }
     underlineBrush {
      bIsDynamicallyLoaded=false
      drawAs=Image
      imageType=NoImage
      mirroring=NoMirror
      resourceName=None
      resourceObject=null
      tiling=NoTile
      imageSize=<32 32>
      margin=<ltrb 0 0 0 0>
      outlineSettings {
       bUseBrushTransparency=false
       roundingType=HalfHeightRadius
       width=0
       color {colorUseRule=UseColor_Specified, specifiedColor=<rgba 0 0 0 0>}
       cornerRadii=<0 0 0 0>
      }
      tintColor {colorUseRule=UseColor_Specified, specifiedColor=<rgba 1 1 1 1>}
      uVRegion {bIsValid=false, max=<0 0>, min=<0 0>}
     }
    }
   }
   slot {class=HorizontalBoxSlot, properties {padding=<ltrb 0 0 8 0>}}
  }
  Switcher_SizeInfo {
   widgetClassPath=&CommonWidgetSwitcherLegacy
   bIsVariable=true
   parent=HorizontalBox_94
   widget=Switcher_SizeInfo
   properties {transitionType=Horizontal}
   slot {class=HorizontalBoxSlot}
  }
  Text_FirstCatch {
   widgetClassPath=&CommonTextBlock
   bIsVariable=true
   parent=Switcher_SizeInfo
   widget=Text_FirstCatch
   properties {
    text=First Catch
    textTransformPolicy=ToUpper
    visibility=Collapsed
    colorAndOpacity {colorUseRule=UseColor_Specified, specifiedColor=<rgba 0 0.5612 1 1>}
    font {
     bForceMonospaced=false
     bMaterialIsStencil=false
     fontMaterial=/Game/Athena/HUD/Materials/M_UI_OutlineShine_White_Manual
     fontObject=&Burbank
     letterSpacing=0
     monospacedWidth=1
     size=16
     skewAmount=0
     typefaceFontName=Burbank-PlayerPower
     outlineSettings {
      bApplyOutlineToDropShadows=false
      bMiteredCorners=false
      bSeparateFillAlpha=false
      outlineMaterial=null
      outlineSize=0
      outlineColor=<rgba 0 0 0 1>
     }
    }
    renderTransformPivot=<0 0.75>
    shadowOffset=<0 0>
   }
   slot {class=WidgetSwitcherSlot}
  }
  Overlay_RelativeSize {widgetClassPath=&Overlay, parent=Switcher_SizeInfo, widget=Overlay_RelativeSize, slot {class=WidgetSwitcherSlot}}
  Image_RelativeSizeMeter {
   widgetClassPath=&Image
   bIsVariable=true
   parent=Overlay_RelativeSize
   widget=Image_RelativeSizeMeter
   properties {
    visibility=SelfHitTestInvisible
    brush {
     bIsDynamicallyLoaded=false
     drawAs=Image
     imageType=NoImage
     mirroring=NoMirror
     resourceName=None
     resourceObject=/Game/Athena/Collections/Art/M_Ui_Collection_Size_Overlay
     tiling=NoTile
     imageSize=<150 14>
     margin=<ltrb 0 0 0 0>
     outlineSettings {
      bUseBrushTransparency=false
      roundingType=HalfHeightRadius
      width=0
      color {colorUseRule=UseColor_Specified, specifiedColor=<rgba 0 0 0 0>}
      cornerRadii=<0 0 0 0>
     }
     tintColor {colorUseRule=UseColor_Specified, specifiedColor=<rgba 1 1 1 1>}
     uVRegion {bIsValid=false, max=<0 0>, min=<0 0>}
    }
   }
   slot {class=OverlaySlot, properties {verticalAlignment=VAlign_Center}}
  }
  Image_RelativeSizePointer {
   widgetClassPath=&Image
   bIsVariable=true
   parent=Overlay_RelativeSize
   widget=Image_RelativeSizePointer
   properties {
    visibility=SelfHitTestInvisible
    brush {
     bIsDynamicallyLoaded=false
     drawAs=Image
     imageType=NoImage
     mirroring=NoMirror
     resourceName=None
     resourceObject=/Game/Athena/UI/ReplayBrowser/M_Ui_Skewed_triangle_collection
     tiling=NoTile
     imageSize=<20 20>
     margin=<ltrb 0 0 0 0>
     outlineSettings {
      bUseBrushTransparency=false
      roundingType=HalfHeightRadius
      width=0
      color {colorUseRule=UseColor_Specified, specifiedColor=<rgba 0 0 0 0>}
      cornerRadii=<0 0 0 0>
     }
     tintColor {colorUseRule=UseColor_Specified, specifiedColor=<rgba 1 1 1 1>}
     uVRegion {bIsValid=false, max=<0 0>, min=<0 0>}
    }
    renderTransform {angle=0, scale=<0.8 -0.8>, shear=<-12 0>, translation=<-11 20>}
    renderTransformPivot=<0.5 0>
   }
   slot {class=OverlaySlot, properties {verticalAlignment=VAlign_Bottom}}
  }
  Border_Leader {
   widgetClassPath=&CommonBorder
   parent=VerticalBox_0
   widget=Border_Leader
   properties {clipping=ClipToBounds, desiredSizeScale=<1 0>, padding=<ltrb 0 0 0 0>}
   slot {class=VerticalBoxSlot, properties {verticalAlignment=VAlign_Bottom}}
  }
  Switcher_LeaderType {
   widgetClassPath=&CommonWidgetSwitcherLegacy
   bIsVariable=true
   parent=Border_Leader
   widget=Switcher_LeaderType
   properties {transitionDuration=0}
   slot {class=BorderSlot, properties {padding=<ltrb 0 0 0 0>}}
  }
  HorizontalBox_0 {widgetClassPath=&HorizontalBox, parent=Switcher_LeaderType, widget=HorizontalBox_0, slot {class=WidgetSwitcherSlot}}
  Text_LeaderName_2 {
   widgetClassPath=&CommonTextBlock
   parent=HorizontalBox_0
   widget=Text_LeaderName_2
   properties {
    text="##1"
    textTransformPolicy=ToUpper
    font {
     bForceMonospaced=false
     bMaterialIsStencil=false
     fontMaterial=null
     fontObject=&Burbank
     letterSpacing=0
     monospacedWidth=1
     size=16
     skewAmount=0
     typefaceFontName=Burbank-PlayerPower
     outlineSettings {
      bApplyOutlineToDropShadows=false
      bMiteredCorners=false
      bSeparateFillAlpha=false
      outlineMaterial=null
      outlineSize=0
      outlineColor=<rgba 0 0 0 1>
     }
    }
    shadowOffset=<0 0>
   }
   slot {class=HorizontalBoxSlot, properties {padding=<ltrb 0 0 8 0>}}
  }
  Text_LeaderName {
   widgetClassPath=&CommonTextBlock
   bInherited=true
   parent=HorizontalBox_0
   widget=Text_LeaderName
   properties {
    text="{Leader Name}"
    textTransformPolicy=ToUpper
    font {
     bForceMonospaced=false
     bMaterialIsStencil=false
     fontMaterial=null
     fontObject=&Burbank
     letterSpacing=0
     monospacedWidth=1
     size=16
     skewAmount=0
     typefaceFontName=Burbank-PlayerPower
     outlineSettings {
      bApplyOutlineToDropShadows=false
      bMiteredCorners=false
      bSeparateFillAlpha=false
      outlineMaterial=null
      outlineSize=0
      outlineColor=<rgba 0 0 0 1>
     }
    }
    shadowOffset=<0 0>
   }
   slot {class=HorizontalBoxSlot, properties {padding=<ltrb 0 0 8 0>}}
  }
  RichText_LeaderLength {
   widgetClassPath=&CommonRichTextBlock
   bInherited=true
   parent=HorizontalBox_0
   widget=RichText_LeaderLength
   properties {
    bOverrideDefaultStyle=true
    text="{(88.8<text Size=\"14\"> cm</>)}"
    defaultTextStyleOverride {
     overflowPolicy=Clip
     transformPolicy=None
     colorAndOpacity {colorUseRule=UseColor_Specified, specifiedColor=<rgba 1 1 1 1>}
     font {
      bForceMonospaced=false
      bMaterialIsStencil=false
      fontMaterial=&M_UI_OutlineShine_White_Manual
      fontObject=&Burbank
      letterSpacing=0
      monospacedWidth=1
      size=16
      skewAmount=0
      typefaceFontName=Burbank-PlayerPower
      outlineSettings {
       bApplyOutlineToDropShadows=false
       bMiteredCorners=false
       bSeparateFillAlpha=false
       outlineMaterial=null
       outlineSize=0
       outlineColor=<rgba 0 0 0 1>
      }
     }
     highlightColor {colorUseRule=UseColor_Specified, specifiedColor=<rgba 0 0 0 1>}
     highlightShape {
      bIsDynamicallyLoaded=false
      drawAs=Image
      imageType=NoImage
      mirroring=NoMirror
      resourceName=None
      resourceObject=null
      tiling=NoTile
      imageSize=<32 32>
      margin=<ltrb 0 0 0 0>
      outlineSettings {
       bUseBrushTransparency=false
       roundingType=HalfHeightRadius
       width=0
       color {colorUseRule=UseColor_Specified, specifiedColor=<rgba 0 0 0 0>}
       cornerRadii=<0 0 0 0>
      }
      tintColor {colorUseRule=UseColor_Specified, specifiedColor=<rgba 1 1 1 1>}
      uVRegion {bIsValid=false, max=<0 0>, min=<0 0>}
     }
     selectedBackgroundColor {colorUseRule=UseColor_Foreground, specifiedColor=<rgba 1 0 1 1>}
     shadowColorAndOpacity=<rgba 0 0 0 1>
     shadowOffset=<0 0>
     strikeBrush {
      bIsDynamicallyLoaded=false
      drawAs=Image
      imageType=NoImage
      mirroring=NoMirror
      resourceName=None
      resourceObject=null
      tiling=NoTile
      imageSize=<32 32>
      margin=<ltrb 0 0 0 0>
      outlineSettings {
       bUseBrushTransparency=false
       roundingType=HalfHeightRadius
       width=0
       color {colorUseRule=UseColor_Specified, specifiedColor=<rgba 0 0 0 0>}
       cornerRadii=<0 0 0 0>
      }
      tintColor {colorUseRule=UseColor_Specified, specifiedColor=<rgba 1 1 1 1>}
      uVRegion {bIsValid=false, max=<0 0>, min=<0 0>}
     }
     underlineBrush {
      bIsDynamicallyLoaded=false
      drawAs=Image
      imageType=NoImage
      mirroring=NoMirror
      resourceName=None
      resourceObject=null
      tiling=NoTile
      imageSize=<32 32>
      margin=<ltrb 0 0 0 0>
      outlineSettings {
       bUseBrushTransparency=false
       roundingType=HalfHeightRadius
       width=0
       color {colorUseRule=UseColor_Specified, specifiedColor=<rgba 0 0 0 0>}
       cornerRadii=<0 0 0 0>
      }
      tintColor {colorUseRule=UseColor_Specified, specifiedColor=<rgba 1 1 1 1>}
      uVRegion {bIsValid=false, max=<0 0>, min=<0 0>}
     }
    }
   }
   slot {class=HorizontalBoxSlot}
  }
  Text_LeaderName_1 {
   widgetClassPath=&CommonTextBlock
   parent=Switcher_LeaderType
   widget=Text_LeaderName_1
   properties {
    text="##1 of your friends!"
    textTransformPolicy=ToUpper
    colorAndOpacity {colorUseRule=UseColor_Specified, specifiedColor=<rgba 1 0.6622 0 1>}
    font {
     bForceMonospaced=false
     bMaterialIsStencil=false
     fontMaterial=&M_UI_OutlineShine_White_Manual
     fontObject=&Burbank
     letterSpacing=0
     monospacedWidth=1
     size=15
     skewAmount=0
     typefaceFontName=BurbankRegular-Black
     outlineSettings {
      bApplyOutlineToDropShadows=false
      bMiteredCorners=false
      bSeparateFillAlpha=false
      outlineMaterial=null
      outlineSize=0
      outlineColor=<rgba 0 0 0 1>
     }
    }
    shadowOffset=<0 0>
   }
   slot {class=WidgetSwitcherSlot, properties {horizontalAlignment=HAlign_Center}}
  }
  Overlay_NewType {
   widgetClassPath=&Overlay
   bIsVariable=true
   parent=Overlay_TopContainer
   widget=Overlay_NewType
   properties {renderTransform {angle=0, scale=<1 1>, shear=<0 0>, translation=<25 -12>}}
   slot {class=OverlaySlot, properties {verticalAlignment=VAlign_Bottom}}
  }
  Image_textglow {
   widgetClassPath=&Image
   parent=Overlay_NewType
   widget=Image_textglow
   properties {
    visibility=HitTestInvisible
    brush {
     bIsDynamicallyLoaded=false
     drawAs=Image
     imageType=NoImage
     mirroring=NoMirror
     resourceName=None
     resourceObject=&M_UI_TextGlow_Add
     tiling=NoTile
     imageSize=<82.1221 71.9925>
     margin=<ltrb 0 0 0 0>
     outlineSettings {
      bUseBrushTransparency=false
      roundingType=HalfHeightRadius
      width=0
      color {colorUseRule=UseColor_Specified, specifiedColor=<rgba 0 0 0 0>}
      cornerRadii=<0 0 0 0>
     }
     tintColor {colorUseRule=UseColor_Specified, specifiedColor=<rgba 0 0.1421 1 1>}
     uVRegion {bIsValid=false, max=<0 0>, min=<0 0>}
    }
    renderTransform {angle=0, scale=<1 1>, shear=<0 0>, translation=<1 -12>}
   }
   slot {class=OverlaySlot, properties {horizontalAlignment=HAlign_Center, verticalAlignment=VAlign_Center, padding=<ltrb -75 -75 -75 -75>}}
  }
  Image_star {
   widgetClassPath=&Image
   parent=Overlay_NewType
   widget=Image_star
   properties {
    visibility=HitTestInvisible
    brush {
     bIsDynamicallyLoaded=false
     drawAs=Image
     imageType=NoImage
     mirroring=NoMirror
     resourceName=None
     resourceObject=/Game/UI/InGame/Quest/Materials/M_UI_SparkStar
     tiling=NoTile
     imageSize=<128 128>
     margin=<ltrb 0 0 0 0>
     outlineSettings {
      bUseBrushTransparency=false
      roundingType=HalfHeightRadius
      width=0
      color {colorUseRule=UseColor_Specified, specifiedColor=<rgba 0 0 0 0>}
      cornerRadii=<0 0 0 0>
     }
     tintColor {colorUseRule=UseColor_Specified, specifiedColor=<rgba 1 1 1 1>}
     uVRegion {bIsValid=false, max=<0 0>, min=<0 0>}
    }
   }
   slot {class=OverlaySlot, properties {horizontalAlignment=HAlign_Center, verticalAlignment=VAlign_Center, padding=<ltrb -64 -64 -64 -64>}}
  }
  VisibilitySwitcher_ToastType {
   widgetClassPath=&CommonVisibilitySwitcher
   bIsVariable=true
   parent=Overlay_NewType
   widget=VisibilitySwitcher_ToastType
   slot {class=OverlaySlot}
  }
  Text_New {
   widgetClassPath=&CommonTextBlock
   parent=VisibilitySwitcher_ToastType
   widget=Text_New
   properties {
    justification=Center
    text="New!"
    textTransformPolicy=ToUpper
    font {
     bForceMonospaced=false
     bMaterialIsStencil=false
     fontMaterial=null
     fontObject=&Burbank
     letterSpacing=0
     monospacedWidth=1
     size=20
     skewAmount=0
     typefaceFontName=BurbankRegular-Black
     outlineSettings {
      bApplyOutlineToDropShadows=true
      bMiteredCorners=false
      bSeparateFillAlpha=false
      outlineMaterial=null
      outlineSize=1
      outlineColor=<rgba 0 0.0326 1 1>
     }
    }
    renderTransform {angle=-10, scale=<1 1>, shear=<0 0>, translation=<0 0>}
    shadowColorAndOpacity=<rgba 0 0.0326 1 1>
    shadowOffset=<1.5 1.5>
   }
   slot {
    class=CommonVisibilitySwitcherSlot
    properties {horizontalAlignment=HAlign_Center, verticalAlignment=VAlign_Center, padding=<ltrb -200 0 -200 0>}
   }
  }
  Text_New_1 {
   widgetClassPath=&CommonTextBlock
   parent=VisibilitySwitcher_ToastType
   widget=Text_New_1
   properties {
    justification=Center
    lineHeightPercentage=0.7
    text="Personal\r\nBest!"
    textTransformPolicy=ToUpper
    font {
     bForceMonospaced=false
     bMaterialIsStencil=false
     fontMaterial=null
     fontObject=&Burbank
     letterSpacing=0
     monospacedWidth=1
     size=18
     skewAmount=0
     typefaceFontName=Burbank-PlayerPower
     outlineSettings {
      bApplyOutlineToDropShadows=true
      bMiteredCorners=false
      bSeparateFillAlpha=false
      outlineMaterial=null
      outlineSize=1
      outlineColor=<rgba 0 0.0326 1 1>
     }
    }
    renderTransform {angle=-10, scale=<1 1>, shear=<0 0>, translation=<0 0>}
    shadowColorAndOpacity=<rgba 0 0.0326 1 1>
    shadowOffset=<1.5 1.5>
   }
   slot {
    class=CommonVisibilitySwitcherSlot
    properties {horizontalAlignment=HAlign_Center, verticalAlignment=VAlign_Top, padding=<ltrb -200 0 -200 0>}
   }
  }
  Text_New_2 {
   widgetClassPath=&CommonTextBlock
   parent=VisibilitySwitcher_ToastType
   widget=Text_New_2
   properties {
    justification=Center
    lineHeightPercentage=0.7
    text="New\r\nRecord!"
    textTransformPolicy=ToUpper
    font {
     bForceMonospaced=false
     bMaterialIsStencil=false
     fontMaterial=null
     fontObject=&Burbank
     letterSpacing=0
     monospacedWidth=1
     size=18
     skewAmount=0
     typefaceFontName=Burbank-PlayerPower
     outlineSettings {
      bApplyOutlineToDropShadows=true
      bMiteredCorners=false
      bSeparateFillAlpha=false
      outlineMaterial=null
      outlineSize=1
      outlineColor=<rgba 0 0.0326 1 1>
     }
    }
    renderTransform {angle=-10, scale=<1 1>, shear=<0 0>, translation=<0 0>}
    shadowColorAndOpacity=<rgba 0 0.0326 1 1>
    shadowOffset=<1.5 1.5>
   }
   slot {
    class=CommonVisibilitySwitcherSlot
    properties {horizontalAlignment=HAlign_Center, verticalAlignment=VAlign_Center, padding=<ltrb -200 0 -200 0>}
   }
  }
  Image_textglow_2 {
   widgetClassPath=&Image
   parent=Overlay_NewType
   widget=Image_textglow_2
   properties {
    visibility=SelfHitTestInvisible
    brush {
     bIsDynamicallyLoaded=false
     drawAs=Image
     imageType=NoImage
     mirroring=NoMirror
     resourceName=None
     resourceObject=&M_UI_TextGlow_Add
     tiling=NoTile
     imageSize=<16 10>
     margin=<ltrb 0 0 0 0>
     outlineSettings {
      bUseBrushTransparency=false
      roundingType=HalfHeightRadius
      width=0
      color {colorUseRule=UseColor_Specified, specifiedColor=<rgba 0 0 0 0>}
      cornerRadii=<0 0 0 0>
     }
     tintColor {colorUseRule=UseColor_Specified, specifiedColor=<rgba 0.0037 0.3613 1 1>}
     uVRegion {bIsValid=false, max=<0 0>, min=<0 0>}
    }
    renderTransform {angle=-7, scale=<1 1>, shear=<0 0>, translation=<0 0>}
   }
   slot {class=OverlaySlot, properties {horizontalAlignment=HAlign_Fill, verticalAlignment=VAlign_Fill, padding=<ltrb -50 5 -50 5>}}
  }
  TutorialHighlight {widgetClassPath=&TutorialHighlight}
 ]
}
