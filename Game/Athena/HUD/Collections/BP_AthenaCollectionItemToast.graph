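# Widget Blueprint for the Athena HUD collection item toast, built on the native AthenaCollectionItemToastSwitcher class.
# A size box with a minimum desired size holds an inherited visibility switcher, aligned top-center, with separate Fishing and Character toast widgets as its pages.
# The Character toast carries a collected-toast HUD element tag.
# The Blueprint has no graph logic or variables, so any switching logic comes from outside it, presumably the parent class.
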WidgetBlueprint /Game/Athena/HUD/Collections/BP_AthenaCollectionItemToast
graphs [
 EventGraph {links [], nodes []}
 Sequencer Events {links [], nodes []}
]
variables []
widgets {
 info {inheritedWidgetCount=2, parentClass=/Script/FortniteUI.AthenaCollectionItemToastSwitcher, rootWidgetClass=&SizeBox, widgetCount=5}
 widgets [
  SizeBox_0 {
   widgetClassPath=&SizeBox
   widget=SizeBox_0
   properties {bOverride_MinDesiredHeight=true, bOverride_MinDesiredWidth=true, minDesiredHeight=150, minDesiredWidth=600}
  }
  WidgetSwitcher_Toasts {
   widgetClassPath=&CommonVisibilitySwitcher
   bInherited=true
   bIsVariable=true
   parent=SizeBox_0
   widget=WidgetSwitcher_Toasts
   slot {class=SizeBoxSlot, properties {horizontalAlignment=HAlign_Center, verticalAlignment=VAlign_Top}}
  }
  BP_AthenaCollectionItemToast_Fishing {
   widgetClassPath=&/BP_AthenaCollectionItemToast_Fishing
   parent=WidgetSwitcher_Toasts
   widget=BP_AthenaCollectionItemToast_Fishing
   slot {class=CommonVisibilitySwitcherSlot}
  }
  BP_AthenaCollectionItemToast_Character {
   widgetClassPath=&/BP_AthenaCollectionItemToast_Character
   parent=WidgetSwitcher_Toasts
   widget=BP_AthenaCollectionItemToast_Character
   properties {elementTag {tagName=HUD.Athena.Collections.CollectedToast}}
   slot {class=CommonVisibilitySwitcherSlot}
  }
  TutorialHighlight {widgetClassPath=&TutorialHighlight}
 ]
}
